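# Translucent unlit Niagara sprite material for the pixel aura, drawing animated, blocky noise patches tinted by particle color.
# UVs are quantized by a dynamic pixelate parameter, offset in U per particle at random, then panned across a grayscale noise texture.
# The dynamic dissolve value is subtracted from the noise, which is then contrast-boosted and multiplied by particle alpha and depth fade.
# Particle RGB feeds a shared emissive adjustment function.
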Material /SparksCosmetics/Items/Auras/Assets/NS_Aura_Pixel/M_Aura_Pixel_AnimatedPixelSprites
graph {
 properties {
  bAutomaticallySetUsageInEditor=false
  bUsedWithNiagaraSprites=true
  blendMode=BLEND_Translucent
  shadingModel=MSM_Unlit
  usedShadingModels=Unlit
 }
 nodes [
  n1 = ParticleColor()
  MaterialFunctionCall(Emissive=@n1.RGB) {materialFunction=&MF_FX_EmissiveAdjustments}
  n2 = TextureCoordinate()
  n3 = DynamicParameter() {defaultValue=<rgba 0 3 1 1>, paramNames [dissolve,pixelate,Param3,Param4]}
  n4 = Divide(A=Floor(None=Multiply(A=@n2, B=@n3.pixelate) {constB=2}), B=@n3.pixelate)
  n5 = AppendVector(A=ParticleRandom(), B=Constant())
  n6 = ScalarParameter() {defaultValue=0.1, parameterName=pan u, sortPriority=1}
  n7 = ScalarParameter() {defaultValue=0.1, parameterName=pan V, sortPriority=2}
  n8 = TextureSampleParameter2D!(UVs=Panner(Coordinate=Add(A=@n4, B=@n5), Speed=AppendVector(A=@n6, B=@n7)) {speedX=0.1, speedY=0.1}, Level, Apply View MipBias) {
   mipValueMode=TMVM_MipLevel
   parameterName=texture
   samplerType=SAMPLERTYPE_LinearGrayscale
   texture=/Game/Characters/Player/Female/Medium/Bodies/F_MED_SuperNovaTaro/Textures/LowResBlurredNoise
  }
  n9 = Subtract(A=ChannelMaskParameter(None=@n8.RGBA) {parameterName=texture opacity channel}.R, B=@n3.dissolve) {constB=0.1}
  n10 = ScalarParameter() {defaultValue=2, parameterName=contrast, sortPriority=3}
  Multiply(
   A=Multiply(
    A=@n1.A
    B=MaterialFunctionCall(In=@n9, Contrast=@n10) {materialFunction=&CheapContrast}.Result
   )
   B=DepthFade(FadeDistance=ScalarParameter() {defaultValue=25, parameterName=Depth Fade, sortPriority=4})
  )
 ]
}
