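# Niagara system for the Pixel aura cosmetic in SparksCosmetics, wrapping a character in glows, pixel sprites and rising speed lines.
# Emitters follow skeletal mesh bones selected per particle partition; body glows cycle through a system color array over time.
# Pixel emitters smooth positions with interpolation, and floating pixels add stepped-curve offsets for blocky motion.
# IsLargeBody scales size and motion, while IsFrontEnd and time of day adjust brightness.
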NiagaraSystem /SparksCosmetics/Items/Auras/Assets/NS_Aura_Pixel/NS_Aura_Pixel
dependencies {
 usedDataInterfaces [
  &NiagaraDataInterfaceArrayColor
  &NiagaraDataInterfaceArrayFloat
  &NiagaraDataInterfaceArrayInt32
  &NiagaraDataInterfaceArrayFloat3
  &NiagaraDataInterfaceCurve
  &NiagaraDataInterfaceColorCurve
  &NiagaraDataInterfaceVector2DCurve
  &NiagaraDataInterfaceVectorCurve
  &NiagaraDataInterfaceSkeletalMesh
 ]
 usedDynamicInputs [
  &MakeCustomLinearColorFromBool
  &RandomBool
  &VectorFromCurve
 ]
 usedModules [
  &self:SystemSpawnScript.SystemScriptSource.SystemScriptGraph.NiagaraNodeAssignment_0.SetVariables_#1
  &self:SystemSpawnScript.SystemScriptSource.SystemScriptGraph.NiagaraNodeAssignment_1.SetVariables_#2
  &self:SystemSpawnScript.SystemScriptSource.SystemScriptGraph.NiagaraNodeAssignment_3.SetVariables_#3
  &SystemState
 ]
}
emitters {
 Bright_Glow_backer {
  emitterUpdateScript [
   &EmitterState()
   &SpawnBurst_Instantaneous(Use Spawn Probability=false, UseLoopCountLimit=true)
  ]
  particleSpawnScript [&InitializeParticle(
    Color=<rgba 3 0.5457 0 0.7>
    Lifetime=3
    Sprite Size Mode=::NewEnumerator3
    Uniform Sprite Size=&Multiply_Float(A=135, B=&MakeCustomFloatFromBool(Bool=$User.IsLargeBody, True Float=1.4, False Float=1))
    UsePositionOffset=false
    Write Material Random=true
   ) {module=InitializeParticle001}]
  particleUpdateScript [
   &ParticleState(
    Kill Particles When Lifetime Has Elapsed=false
    Let Infinitely Lived Particles Die When Emitter Deactivates=true
    Loop Particles Lifetime=true
   )
   &SpriteRotationRate(Rotation Rate=5)
   &SkeletalMeshLocation(
    Bone / Socket Index=&SelectIntFromArray(Direct Array Index=0, Int Selection Array=DataInterface() {propertyValues {IntData [1,2]}})
    Bone Sampling Mode=::NewEnumerator4
    MaskSpawnGroup=false
    Orientation Sampling=::NewEnumerator1
    Position Sampling=::NewEnumerator1
    Sampled Position Offset Coordinate Space:ENiagaraSkelMeshTransforms=::NewEnumerator2
    Skeletal Mesh=$System.Skeletal Mesh
   ) {module=SkeletalMeshLocation003}
   &SkeletalMeshLocation(
    Bone / Socket Index=&SelectIntFromArray(Direct Array Index=1, Int Selection Array=DataInterface() {propertyValues {IntData [1,2]}})
    Bone Sampling Mode=::NewEnumerator4
    MaskSpawnGroup=false
    Orientation Sampling=::NewEnumerator1
    Position Sampling=::NewEnumerator1
    Sampled Position Offset Coordinate Space:ENiagaraSkelMeshTransforms=::NewEnumerator2
    Skeletal Mesh=$System.Skeletal Mesh
   )
   /SetVariables(Particles.Position=&LerpPosition(A=$Particles.SkeletalMeshLocation.SampledPosition, B=$Particles.SkeletalMeshLocation003.SampledPosition, Alpha=0.5)) {isSetParameters=true}
   &CameraOffset(Camera Offset Amount=-20)
   &ScaleColor(
    Scale Alpha=&Multiply_Float(A=1, B=&MakeCustomFloatFromBool(Bool=$User.IsFrontEnd, True Float=0.6, False Float=1))
    ScaleA=true
    ScaleRGB=true
   )
   &DynamicMaterialParameters(
    Index 0 Param 1=2
    Index 0 Param 2=1
    Index 0 Param 4=1
    Module.Index0Float0Write=true
    Module.Index0Float1Write=true
    Module.Index0Float2Write=true
    Module.Index0Float3Write=true
   )
  ]
  rendererClasses [&NiagaraSpriteRendererProperties]
  renderers [{rendererClass=&NiagaraSpriteRendererProperties}]
 }
 Glow_AndDecals {
  emitterUpdateScript [
   &EmitterState()
   &SpawnBurst_Instantaneous(Spawn Count=7, Use Spawn Probability=false, UseLoopCountLimit=true)
  ]
  particleSpawnScript [
   &PartitionParticles(Number Of Partitions=7)
   &InitializeParticle(
    Color=&SelectLinearColorFromArray(
     Array Sampling Mode=::NewEnumerator1
     Direct Array Index=$Particles.PartitionParticles.Partition
     Color Selection Array=DataInterface() {propertyValues {ColorData [<rgba 1 1 1 0.8>, <rgba 1 1 1 0.8>, <rgba 1 1 1 0.8>, <rgba 1 1 1 0.7>, <rgba 1 1 1 0.7>, <rgba 1 1 1 0.5>, <rgba 1 1 1 0.5>]}}
     Random Seed=?
    )
    Lifetime=3
    Sprite Rotation Mode=::NewEnumerator0
    Sprite Size Mode=::NewEnumerator3
    Uniform Sprite Size=&Multiply_Float(
     A=&SelectFloatFromArray(
      Direct Array Index=&SpawnGroup()
      Float Selection Array=DataInterface() {propertyValues {FloatData [40,40,40,30,30,20,20]}}
     )
     B=&MakeCustomFloatFromBool(Bool=$User.IsLargeBody, True Float=1.5, False Float=1)
    )
    UsePositionOffset=false
    Write Material Random=true
   ) {module=InitializeParticle001}
   /SetVariables(Particles.glowScale:Vector2f=&Vector2DFromFloat(Value=&Multiply_Float(
      A=&SelectFloatFromArray(
       Direct Array Index=$Particles.PartitionParticles.Partition
       Float Selection Array=DataInterface() {propertyValues {FloatData [80,85,85,65,65,50,50]}}
      )
      B=&MakeCustomFloatFromBool(Bool=$User.IsLargeBody, True Float=1.2, False Float=1)
     ))) {isSetParameters=true}
   /SetVariables(Particles.DecalColor:LinearColor=&Multiply_LinearColorByFloat(
     Linear Color=&SelectLinearColorFromArray(
      Array Sampling Mode=::NewEnumerator1
      Direct Array Index=$Particles.PartitionParticles.Partition
      Color Selection Array=DataInterface() {propertyValues {ColorData [
         <rgba 1 0.591 0.4167 1>
         <rgba 1 0.591 0.4167 1>
         <rgba 1 0.591 0.4167 1>
         <rgba 1 0.5198 0.3151 1>
         <rgba 1 0.5198 0.3151 1>
         <rgba 1 0.4358 0.1953 1>
         <rgba 1 0.4358 0.1953 1>
        ]}}
      Random Seed=?
     )
     Float=&MakeCustomFloatFromBool(
      Bool=$User.IsFrontEnd
      True Float=2
      False Float=&Multiply_Float(A=4, B=&FNTimeOfDayScalar(Night Scale=0.5, Morning Scale=0.6, Day Scale=0.6, Evening Scale=0.5))
     )
    )) {isSetParameters=true}
   /SetVariables(Particles.glowCameraOffset=-15) {isSetParameters=true}
   &CameraOffset(Camera Offset Amount=5)
  ]
  particleUpdateScript [
   &ParticleState(
    Kill Particles When Lifetime Has Elapsed=false
    Let Infinitely Lived Particles Die When Emitter Deactivates=true
    Loop Particles Lifetime=true
   )
   &SkeletalMeshLocation(
    Bone / Socket Index=&SelectIntFromArray(Direct Array Index=$Particles.PartitionParticles.Partition, Int Selection Array=DataInterface() {propertyValues {IntData [0,1,2,3,4,5,6]}})
    Bone Sampling Mode=::NewEnumerator4
    MaskSpawnGroup=false
    Orientation Sampling=::NewEnumerator1
    Sampled Position Offset=&SelectVectorFromArray(
     Direct Array Index=$Particles.PartitionParticles.Partition
     Vector Selection Array=DataInterface() {propertyValues {FloatData [<0 0 -5>, <0 0 0>, <0 0 0>, <0 0 0>, <0 0 0>, <0 0 0>, <0 0 0>]}}
    )
    Sampled Position Offset Coordinate Space:ENiagaraSkelMeshTransforms=::NewEnumerator2
    Skeletal Mesh=$System.Skeletal Mesh
   ) {module=SkeletalMeshLocation003}
   &DynamicMaterialParameters(
    Index 0 Param 1=&SelectFloatFromArray(
     Direct Array Index=$Particles.PartitionParticles.Partition
     Float Selection Array=DataInterface() {propertyValues {FloatData [5,3,3,4,4,4,4]}}
    )
    Index 0 Param 3=1
    Module.Param0WriteEnabled=true
    Module.Param1WriteEnabled=true
    Module.Param2WriteEnabled=false
    Module.Param3WriteEnabled=false
   )
   &SpriteRotationRate(Rotation Rate=10)
   &InterpolateOverTime(
    Rate Of Change=3
    Target Color Value:LinearColor=&SelectLinearColorFromArray(
     Array Sampling Mode=::NewEnumerator1
     Direct Array Index=&Add_Integer(A=$Particles.PartitionParticles.Partition, B=&MakeIntFromFloat(Float=$Engine.Time))
     Direct Array Mode=::NewEnumerator1
     Color Selection Array=$System.Glow Colors
     Random Seed=?
    )
    Variable Type=::NewEnumerator4
   )
   &ScaleColor(
    Scale Alpha=&Multiply_Float(
     A=&FNTimeOfDayScalar(Night Scale=0.4, Morning Scale=0.7, Day Scale=0.8, Evening Scale=0.5)
     B=&MakeCustomFloatFromBool(Bool=$User.IsFrontEnd, True Float=0.6, False Float=1)
    )
    Scale RGB=&MakeVectorFromLinearColorRGB(Color=$Particles.InterpolateOverTime.InterpolateOverTime_LinearColor.Moving Average)
    ScaleA=true
    ScaleRGB=true
   )
   /SetVariables(Particles.DecalColor:LinearColor=&Multiply_LinearColorByFloat(
     Linear Color=$Particles.InterpolateOverTime.InterpolateOverTime_LinearColor.Moving Average
     Float=&MakeCustomFloatFromBool(
      Bool=$User.IsFrontEnd
      True Float=5
      False Float=&Multiply_Float(A=6, B=&FNTimeOfDayScalar(Night Scale=0.5, Morning Scale=0.6, Day Scale=0.6, Evening Scale=0.5))
     )
    )) {enabled=false, isSetParameters=true}
  ]
  rendererClasses [&NiagaraSpriteRendererProperties]
  renderers [
   {rendererClass=&NiagaraSpriteRendererProperties}
   {rendererClass=&NiagaraSpriteRendererProperties, rendererIndex=1}
   {rendererClass=&NiagaraSpriteRendererProperties, rendererIndex=2}
  ]
 }
 Pixel_small_Particles {
  emitterUpdateScript [
   &EmitterState()
   &SpawnRate(SpawnRate=&RandomRangeFloat(Minimum=4, Maximum=8), Use Spawn Probability=false)
  ]
  particleSpawnScript [
   &PartitionParticles(Number Of Partitions=5)
   &InitializeParticle(
    Color=<rgba 0 1 0.9046 2>
    Lifetime Max=0.6
    Lifetime Min=0.4
    Lifetime Mode=::NewEnumerator1
    Sprite Size Mode=::NewEnumerator1
    Uniform Sprite Size Max=6
    Uniform Sprite Size Min=3
    UsePositionOffset=false
    Write Material Random=true
   )
   &CameraOffset(Camera Offset Amount=-25)
   &ShapeLocation(
    Apply To Particle Position=::NewEnumerator1
    MaskSpawnGroup=false
    Offset=<0 0 10>
    Offset Mode=::NewEnumerator0
    Override Seed=false
    Shape Origin=&VectorFromFloat(Value=0)
    Sphere Radius=1
    UseOwnerScale=true
   )
   &AddVelocity(Velocity=<0 0 85>) {module=AddVelocity001}
  ]
  particleUpdateScript [
   &ParticleState()
   &Drag(Drag=2, Use Linear Drag=true, Use Rotational Drag=true)
   &ScaleSpriteSize(
    Uniform Curve Scale=0.6
    Uniform Curve Sprite Scale=DataInterface() {propertyValues {Curve {keys [
        {arriveTangent=8.9937, interpMode=RCIM_Cubic, leaveTangent=8.9937, tangentMode=RCTM_User, time=0, value=0}
        {arriveTangentWeight=0.1601, interpMode=RCIM_Cubic, leaveTangentWeight=0.2668, tangentMode=RCTM_User, time=0.2, value=1}
        {arriveTangentWeight=0.2906, interpMode=RCIM_Cubic, leaveTangentWeight=0.1163, tangentMode=RCTM_User, time=0.7, value=1}
        {arriveTangent=-6.7083, interpMode=RCIM_Cubic, leaveTangent=-6.7083, tangentMode=RCTM_User, time=1, value=0}
       ]}}}
   )
   &ScaleColor(
    Scale Alpha=&FloatFromCurve(FloatCurve=DataInterface() {propertyValues {Curve {keys [<at 0.8 1>,<at 1 0>]}}}, Scale Curve=2)
    ScaleA=true
    ScaleRGB=true
   )
   &SkeletalMeshLocation(
    Bone / Socket Index=&SelectIntFromArray(Direct Array Index=$Particles.PartitionParticles.Partition, Int Selection Array=DataInterface() {propertyValues {IntData [0,1,2,3,4]}})
    Bone Sampling Mode=::NewEnumerator4
    MaskSpawnGroup=false
    Orientation Sampling=::NewEnumerator1
    Position Sampling=::NewEnumerator1
    Sampled Position Offset=&AddVectorToPosition(Position=$StackContext.ShapeLocation.ShapePosition, Vector=<0 -5 5>)
    Sampled Position Offset Coordinate Space:ENiagaraSkelMeshTransforms=::NewEnumerator1
    Skeletal Mesh=$System.Skeletal Mesh
   ) {module=SkeletalMeshLocation003}
   &SolveForcesAndVelocity(Clamp Velocity=false, Limit Acceleration=false)
   &DynamicMaterialParameters(
    Index 0 Param 1=2
    Module.Index0Param0Write=true
    Module.Index0Param1Write=true
    Module.Index0Param2Write=true
    Module.Index0Param3Write=true
   )
   /SetVariables(Particles.PositionOffset:NiagaraPosition=&AddVectorToPosition(
     Position=$Particles.Position
     Vector=&MakeVector(
      X=&FloatFromCurve(
       FloatCurve=DataInterface() {propertyValues {Curve {keys [<at 0 0 RCIM_Constant>,<at 0.1 0.5 RCIM_Constant>,<at 0.2 0.75 RCIM_Constant>,<at 0.3 0.875 RCIM_Constant>,<at 0.6 0.25 RCIM_Constant>,<at 0.7 0 RCIM_Constant>,<at 0.8 -0.625 RCIM_Constant>,<at 0.9 -1 RCIM_Constant>]}}}
       Scale Curve=&SelectFloatFromArray(
        Array Sampling Mode=::NewEnumerator0
        Direct Array Index=?
        Float Selection Array=DataInterface() {propertyValues {FloatData [0,3,-2,0]}}
        Random Seed=0
       )
      )
      Y=0
      Z=&FloatFromCurve(
       FloatCurve=DataInterface() {propertyValues {Curve {keys [<at 0 0 RCIM_Constant>,<at 0.1 -0.25 RCIM_Constant>,<at 0.2 -0.125 RCIM_Constant>,<at 0.4 -0.125 RCIM_Constant>,<at 0.5 0.375 RCIM_Constant>,<at 0.6 0.625 RCIM_Constant>,<at 0.7 0.75 RCIM_Constant>,<at 0.8 1.375 RCIM_Constant>]}}}
       Scale Curve=&SelectFloatFromArray(
        Array Sampling Mode=::NewEnumerator0
        Direct Array Index=?
        Float Selection Array=DataInterface() {propertyValues {FloatData [0,-1,2]}}
        Random Seed=0
       )
      )
     )
    )) {enabled=false, isSetParameters=true}
  ]
  rendererClasses [&NiagaraSpriteRendererProperties]
  renderers [{rendererClass=&NiagaraSpriteRendererProperties}]
 }
 Pixels_Animated_Floating {
  emitterUpdateScript [
   &EmitterState()
   &SpawnRate(SpawnRate=4, Use Spawn Probability=false)
  ]
  particleSpawnScript [
   &PartitionParticles(Number Of Partitions=5)
   &SkeletalMeshLocation(
    Bone / Socket Index=7
    Bone Sampling Mode=::NewEnumerator4
    MaskSpawnGroup=false
    Position Sampling=::NewEnumerator2
    Sampled Position Offset Coordinate Space:ENiagaraSkelMeshTransforms=::NewEnumerator0
    Skeletal Mesh=$System.Skeletal Mesh
   ) {module=SkeletalMeshLocation001}
   &InitializeParticle(
    Lifetime Max=1
    Lifetime Min=0.7
    Lifetime Mode=::NewEnumerator1
    Sprite Size=&Multiply_Vector2D_ByFloat(
     Float=&MakeCustomFloatFromBool(Bool=$User.IsLargeBody, True Float=1.2, False Float=1)
     Vector2D=&Multiply_Vector2D_ByFloat(
      Float=&RandomRangeFloat(Minimum=6, Maximum=10)
      Vector2D=&RandomRangeVector2D(Minimum=<1 1>, Maximum=<1.25 1.5>)
     )
    )
    Sprite Size Mode=::NewEnumerator4
    UsePositionOffset=true
    Write Material Random=true
   )
   /SetVariables() {isSetParameters=true}
   &CameraOffset(Camera Offset Amount=-30)
   &ShapeLocation(
    Apply To Particle Position=::NewEnumerator1
    Box Size=<10 1 2>
    MaskSpawnGroup=false
    Override Seed=false
    Shape Origin=&VectorFromFloat(Value=0)
    Shape Primitive=::NewEnumerator2
    Surface Only=false
    UseOwnerScale=true
   )
   &AddVelocity(
    Rotation Coordinate Space=::Simulation
    Velocity=&RandomRangeVector(Minimum=<0 0 300>, Maximum=<0 0 375>)
    Velocity Speed Scale=&MakeCustomFloatFromBool(Bool=$User.IsLargeBody, True Float=1.5, False Float=1)
   )
   &AddVelocity(
    Module.Use Falloff Distance=false
    Override Seed=false
    Velocity Mode=::NewEnumerator1
    Velocity Origin=$Output.SkeletalMeshLocation001.SampledPositionWithoutOffset
    Velocity Speed=&RandomRangeFloat(Minimum=25, Maximum=50)
   ) {module=AddVelocity001}
  ]
  particleUpdateScript [
   &ParticleState()
   &ScaleSpriteSize(Uniform Curve Sprite Scale=DataInterface() {propertyValues {Curve {keys [{tangentMode=RCTM_User, time=0, value=0}, <at 0.2 0.8>, <at 0.8 0.9>, {tangentMode=RCTM_User, time=1, value=1.3}]}}})
   &Drag(Drag=2.5, Use Linear Drag=true, Use Rotational Drag=true)
   &SkeletalMeshLocation(
    Bone / Socket Index=&SelectIntFromArray(Direct Array Index=$Particles.PartitionParticles.Partition, Int Selection Array=DataInterface() {propertyValues {IntData [0,1,2,3,4]}})
    Bone Sampling Mode=::NewEnumerator4
    MaskSpawnGroup=false
    Orientation Sampling=::NewEnumerator1
    Position Sampling=::NewEnumerator1
    Sampled Position Offset=&AddVectorToPosition(
     Position=$StackContext.ShapeLocation.ShapePosition
     Vector=&Multiply_VectorByFloat(
      Vector=&RandomRangeVector(Minimum=<0 0 10>, Maximum=<0 -15 15>)
      Float=&MakeCustomFloatFromBool(Bool=$User.IsLargeBody, True Float=1.1, False Float=1)
     )
    )
    Sampled Position Offset Coordinate Space:ENiagaraSkelMeshTransforms=::NewEnumerator2
    Skeletal Mesh=$System.Skeletal Mesh
   ) {module=SkeletalMeshLocation003}
   &SolveForcesAndVelocity(Clamp Velocity=false, Limit Acceleration=false)
   &InterpolateOverTime(Rate Of Change=7, Target Position Value:NiagaraPosition=$Output.SolveForcesAndVelocity.Position, Variable Type=::NewEnumerator6)
   &ScaleColor(
    Linear Color Curve=DataInterface() {propertyValues {
      AlphaCurve {keys [<at 0 0>,<at 0.0324 1>,<at 0.9279 1>,<at 1 0>]}
      BlueCurve {keys [<at 0.0382 1>,<at 0.1842 0>,<at 0.5818 0.0173>,<at 0.8528 0.2174>]}
      GreenCurve {keys [<at 0.0382 0.0007>,<at 0.1842 0.0969>,<at 0.5818 0>,<at 0.8528 0>]}
      RedCurve {keys [<at 0.0382 0.5893>,<at 0.1842 1.5>,<at 0.5818 1.1>,<at 0.8528 0.5885>]}
     }}
    Scale Mode=::NewEnumerator2
   )
   /SetVariables(Particles.color2:LinearColor=&MultiplyLinearColor(
     Linear Color=&ColorFromCurve(ColorCurve=DataInterface() {propertyValues {
        AlphaCurve {keys [<at 0 0 RCIM_Cubic>,<at 0.0386 0.9939>,<at 0.885 1>,<at 1 0 RCIM_Cubic>]}
        BlueCurve {keys [
          <at 0.0588 0.3116>
          <at 0.1258 0.7292>
          <at 0.204 1>
          {arriveTangent=-1.209, interpMode=RCIM_Cubic, leaveTangent=-1.209, time=0.413, value=0.8958}
          <at 0.5419 0>
          <at 0.9905 0.0492>
         ]}
        GreenCurve {keys [
          <at 0.0588 0.6823>
          <at 0.1258 0>
          <at 0.204 0.0972>
          {arriveTangent=0.2803, interpMode=RCIM_Cubic, leaveTangent=0.2803, time=0.413, value=0.6975}
          <at 0.5419 0.5052>
          <at 0.9905 0.3177>
         ]}
        RedCurve {keys [<at 0.0588 0>,<at 0.1258 0.5831>,<at 0.204 0>,<at 0.413 0 RCIM_Cubic>,<at 0.5419 0.0813>,<at 0.9905 0>]}
       }})
     Scale Factor=/Niagara/DynamicInputs/TypeConversions/MakeCustomVector4FromBool(Bool:NiagaraBool=$User.IsFrontEnd, True Vector4:Vector4f=<0.6 0.6 0.6 1>, False Vector4:Vector4f=<1 1 1 1>)
    )) {isSetParameters=true}
   &DynamicMaterialParameters(
    Index 0 Param 1=&FloatFromCurve(FloatCurve=DataInterface() {propertyValues {Curve {defaultValue=1, keys [{tangentMode=RCTM_User, time=0.8, value=0.1}, {tangentMode=RCTM_User, time=1, value=0.5}]}}})
    Index 0 Param 2=&FloatFromCurve(FloatCurve=DataInterface() {propertyValues {Curve {keys [<at 0.85 3>,<at 1 5>]}}})
    Module.Index0Float0Write=true
    Module.Index0Float1Write=true
    Module.Index0Float2Write=true
    Module.Index0Float3Write=true
   )
   /SetVariables(Particles.PivotOffset2:Vector2f=/Niagara/DynamicInputs/Add/Add_Vector2(
     A:Vector2f=$Particles.PivotOffset
     B:Vector2f=&MakeVector2D(
      X=&FloatFromCurve(FloatCurve=DataInterface() {propertyValues {Curve {
          defaultValue=1
          keys [
           {arriveTangent=0.4331, interpMode=RCIM_Cubic, leaveTangent=0.4331, tangentMode=RCTM_User, time=0.7, value=0.1}
           {interpMode=RCIM_Cubic, tangentMode=RCTM_User, time=1, value=0.2}
          ]
         }}})
      Y=&FloatFromCurve(FloatCurve=DataInterface() {propertyValues {Curve {
          defaultValue=1
          keys [
           {arriveTangent=0.4564, interpMode=RCIM_Cubic, leaveTangent=0.4564, tangentMode=RCTM_User, time=0.7, value=0.1}
           {arriveTangentWeight=0.3333, interpMode=RCIM_Cubic, tangentMode=RCTM_Break, time=1, value=0.2}
          ]
         }}})
     )
    )) {isSetParameters=true}
   /SetVariables(Particles.Position=$StackContext.InterpolateOverTime.InterpolateOverTime_Position.Moving Average) {isSetParameters=true}
   /SetVariables(Particles.PositionOffset:NiagaraPosition=&AddVectorToPosition(
     Position=$Particles.Position
     Vector=&MakeVector(
      X=&FloatFromCurve(
       FloatCurve=DataInterface() {propertyValues {Curve {keys [<at 0 0 RCIM_Constant>,<at 0.1 0.5 RCIM_Constant>,<at 0.2 1 RCIM_Constant>,<at 0.5 1 RCIM_Constant>,<at 0.6 0.25 RCIM_Constant>,<at 0.7 0 RCIM_Constant>,<at 0.8 -0.625 RCIM_Constant>,<at 0.9 -1 RCIM_Constant>]}}}
       Scale Curve=&SelectFloatFromArray(
        Array Sampling Mode=::NewEnumerator0
        Direct Array Index=?
        Float Selection Array=DataInterface() {propertyValues {FloatData [0,3,-2,0]}}
        Random Seed=0
       )
      )
      Y=0
      Z=&FloatFromCurve(
       FloatCurve=DataInterface() {propertyValues {Curve {keys [<at 0 0 RCIM_Constant>,<at 0.1 0.25 RCIM_Constant>,<at 0.3 0.575 RCIM_Constant>,<at 0.4 0.8 RCIM_Constant>,<at 0.6 1 RCIM_Constant>,<at 0.7 1.2 RCIM_Constant>,<at 0.8 1.375 RCIM_Constant>]}}}
       Scale Curve=&SelectFloatFromArray(
        Array Sampling Mode=::NewEnumerator0
        Direct Array Index=?
        Float Selection Array=DataInterface() {propertyValues {FloatData [2,4,7]}}
        Random Seed=0
       )
      )
     )
    )) {isSetParameters=true}
  ]
  rendererClasses [&NiagaraSpriteRendererProperties]
  renderers [
   {rendererClass=&NiagaraSpriteRendererProperties}
   {rendererClass=&NiagaraSpriteRendererProperties, rendererIndex=1}
  ]
 }
 Pixels_onBody {
  emitterUpdateScript [
   &EmitterState()
   &SpawnRate(SpawnRate=50, Use Spawn Probability=false)
  ]
  particleSpawnScript [
   &PartitionParticles(Number Of Partitions=3)
   &SkeletalMeshLocation(
    Bone / Socket Index=&SelectIntFromArray(Direct Array Index=$Particles.PartitionParticles.Partition, Int Selection Array=DataInterface() {propertyValues {IntData [1,2,3,4]}})
    Bone Sampling Mode=::NewEnumerator4
    MaskSpawnGroup=false
    Orientation Sampling=::NewEnumerator1
    Sampled Position Offset Coordinate Space:ENiagaraSkelMeshTransforms=::NewEnumerator2
    Skeletal Mesh=$System.Skeletal Mesh
   ) {enabled=false, module=SkeletalMeshLocation003}
   &SkeletalMeshLocation(
    Bone / Socket Index=7
    Bone Sampling Mode=::NewEnumerator4
    MaskSpawnGroup=false
    Position Sampling=::NewEnumerator2
    Sampled Position Offset Coordinate Space:ENiagaraSkelMeshTransforms=::NewEnumerator2
    Skeletal Mesh=$System.Skeletal Mesh
   ) {module=SkeletalMeshLocation001}
   &SkeletalMeshLocation(
    Bone / Socket Index=&SelectIntFromArray(Direct Array Index=$Particles.PartitionParticles.Partition, Int Selection Array=$System.companion bone)
    Bone Sampling Mode=::NewEnumerator4
    MaskSpawnGroup=false
    Orientation Sampling=::NewEnumerator1
    Sampled Position Offset Coordinate Space:ENiagaraSkelMeshTransforms=::NewEnumerator2
    Skeletal Mesh=$System.Skeletal Mesh
   ) {enabled=false}
   /SetVariables(
    Particles.bonesOffset=DataInterface() {propertyValues {GpuSyncMode=SyncCpuToGpu, MaxElements=0, FloatData [<-20 0 20>, <25 -5 1>, <-25 5 -1>]}}
    Particles.positionlerp:NiagaraFloat=&RandomRangeFloat(Minimum=0, Maximum=1)
    Particles.postionOffset:Vector3f=&Multiply_VectorByFloat(
     Vector=&MakeCustomVectorFromBool(
      Bool=$User.IsFrontEnd
      True Vector=&RandomRangeVector(Minimum=<0 0 1>, Maximum=<0 0 2>)
      False Vector=<0 -1 1>
     )
     Float=&MakeCustomFloatFromBool(
      Bool=$User.IsLargeBody
      True Float=&RandomRangeFloat(Minimum=9, Maximum=15)
      False Float=&RandomRangeFloat(Minimum=2, Maximum=4)
     )
    )
   ) {isSetParameters=true}
   &InitializeParticle(
    Lifetime Max=0.6
    Lifetime Min=0.3
    Lifetime Mode=::NewEnumerator1
    Sprite Size Mode=::NewEnumerator3
    Uniform Sprite Size=&Multiply_Float(
     A=&RandomRangeFloat(Minimum=1.5, Maximum=7)
     B=&MakeCustomFloatFromBool(Bool=$User.IsLargeBody, True Float=1.2, False Float=1)
    )
    UsePositionOffset=false
    Write Material Random=true
   )
   &SkeletalMeshLocation(
    Bone / Socket Index=&SelectIntFromArray(Direct Array Index=$Particles.PartitionParticles.Partition, Int Selection Array=DataInterface() {propertyValues {IntData [0,1,2]}})
    Bone Sampling Mode=::NewEnumerator4
    MaskSpawnGroup=false
    Orientation Sampling=::NewEnumerator1
    Position Sampling=::NewEnumerator1
    Sampled Position Offset=&LerpPosition(
     A=<0 0 0>
     B=&SelectVectorFromArray(Direct Array Index=$Particles.PartitionParticles.Partition, Vector Selection Array=$Particles.bonesOffset)
     Alpha=$Particles.positionlerp
    )
    Skeletal Mesh=$System.Skeletal Mesh
   ) {module=SkeletalMeshLocation004}
   /SetVariables() {isSetParameters=true}
   &CameraOffset(Camera Offset Amount=&MakeCustomFloatFromBool(
     Bool=&SetBoolByIntComparison(A=$Particles.PartitionParticles.Partition, B=1, Comparison Type=::NewEnumerator7)
     True Float=&RandomRangeFloat(Minimum=-20, Maximum=-15)
     False Float=&SelectFloatFromArray(
      Array Sampling Mode=::NewEnumerator0
      Direct Array Index=?
      Float Selection Array=DataInterface() {propertyValues {FloatData [-20,-20,-20,-20,-15,-15,-10,-10]}}
      Random Seed=0
     )
    ))
   &CameraOffset(Camera Offset Amount=40) {enabled=false, module=CameraOffset001}
   /SetVariables(Particles.cameraoffset2=-20) {enabled=false, isSetParameters=true}
   &AddVelocity(
    Velocity=&MakeCustomVectorFromBool(
     Bool=$User.IsFrontEnd
     True Vector=&RandomRangeVector(Minimum=<0 0 30>, Maximum=<0 0 45>)
     False Vector=&RandomRangeVector(Minimum=<0 0 35>, Maximum=<0 0 50>)
    )
    Velocity Speed Scale=&MakeCustomFloatFromBool(Bool=$User.IsLargeBody, True Float=1, False Float=1)
   ) {module=AddVelocity001}
   &AddVelocity(
    Module.Use Falloff Distance=false
    Override Seed=false
    Velocity Mode=::NewEnumerator1
    Velocity Origin=$Output.SkeletalMeshLocation001.SampledPositionWithoutOffset
    Velocity Speed=&Multiply_Float(
     A=&MakeCustomFloatFromBool(
      Bool=$User.IsFrontEnd
      True Float=&RandomRangeFloat(Minimum=25, Maximum=35)
      False Float=&RandomRangeFloat(Minimum=25, Maximum=45)
     )
     B=&MakeCustomFloatFromBool(Bool=$User.IsLargeBody, True Float=1.2, False Float=1)
    )
   )
  ]
  particleUpdateScript [
   &ParticleState()
   &SkeletalMeshLocation(
    Bone / Socket Index=&SelectIntFromArray(Direct Array Index=$Particles.PartitionParticles.Partition, Int Selection Array=DataInterface() {propertyValues {IntData [0,1,2]}})
    Bone Sampling Mode=::NewEnumerator4
    MaskSpawnGroup=false
    Orientation Sampling=::NewEnumerator1
    Position Sampling=::NewEnumerator1
    Sampled Position Offset=&LerpPosition(
     A=&SelectVectorFromArray(
      Direct Array Index=$Particles.PartitionParticles.Partition
      Vector Selection Array=DataInterface() {propertyValues {FloatData [<-15 0 -15>, <0 0 0>, <0 0 0>]}}
     )
     B=&SelectVectorFromArray(Direct Array Index=$Particles.PartitionParticles.Partition, Vector Selection Array=$Particles.bonesOffset)
     Alpha=$Particles.positionlerp
    )
    Skeletal Mesh=$System.Skeletal Mesh
   ) {module=SkeletalMeshLocation002}
   &Drag(Drag=10, Use Linear Drag=true, Use Rotational Drag=true)
   &SolveForcesAndVelocity(Clamp Velocity=false, Limit Acceleration=false)
   &ScaleSpriteSize(Uniform Curve Sprite Scale=DataInterface() {propertyValues {Curve {
       defaultValue=0.85
       keys [
        {arriveTangent=6.7723, interpMode=RCIM_Cubic, leaveTangent=6.7723, tangentMode=RCTM_User, time=0, value=0.1}
        {arriveTangent=-0.0348, interpMode=RCIM_Cubic, leaveTangent=-0.0348, tangentMode=RCTM_User, time=0.3, value=1}
        {arriveTangent=1.2587, arriveTangentWeight=0.2667, interpMode=RCIM_Cubic, leaveTangent=1.2587, tangentMode=RCTM_User, time=1, value=1.3}
       ]
      }}})
   &InterpolateOverTime(Rate Of Change=30, Target Position Value:NiagaraPosition=$Output.SolveForcesAndVelocity.Position, Variable Type=::NewEnumerator6)
   &ScaleColor(
    Linear Color Curve=DataInterface() {propertyValues {
      AlphaCurve {keys [<at 0 0>,<at 0.123 1>,<at 0.7333 1>,<at 1 0>]}
      BlueCurve {keys [<at 0.0772 1>,<at 0.1911 0>,<at 0.7263 0.0851>,<at 0.9919 0.2975>]}
      GreenCurve {keys [<at 0.0772 0.0007>,<at 0.1911 0.0969>,<at 0.7263 0>,<at 0.9919 0>]}
      RedCurve {keys [<at 0.0772 0.5893>,<at 0.1911 1.5>,<at 0.7263 1>,<at 0.9919 0.849>]}
     }}
    Scale Mode=::NewEnumerator2
   )
   &DynamicMaterialParameters(
    Index 0 Param 1=&FloatFromCurve(
     FloatCurve=DataInterface() {propertyValues {Curve {defaultValue=1, keys [<at 0.5 0.05>, {tangentMode=RCTM_User, time=0.7, value=0.05}, {tangentMode=RCTM_User, time=1, value=1}]}}}
     Scale Curve=&SelectFloatFromArray(
      Array Sampling Mode=::NewEnumerator0
      Direct Array Index=?
      Float Selection Array=DataInterface() {propertyValues {FloatData [0,2,2]}}
      Random Seed=0
     )
    )
    Index 0 Param 2=&FloatFromCurve(FloatCurve=DataInterface() {propertyValues {Curve {
        defaultValue=1
        keys [
         {arriveTangent=8.2781, interpMode=RCIM_Cubic, leaveTangent=8.2781, tangentMode=RCTM_User, time=0.2, value=0}
         {arriveTangent=1.1132, interpMode=RCIM_Cubic, leaveTangent=1.1132, tangentMode=RCTM_User, time=0.8, value=3}
        ]
       }}})
    Module.Index0Float0Write=true
    Module.Index0Float1Write=true
    Module.Index0Float2Write=true
    Module.Index0Float3Write=true
   )
   /SetVariables(Particles.color2:LinearColor=&MultiplyLinearColor(
     Linear Color=&ColorFromCurve(ColorCurve=DataInterface() {propertyValues {
        AlphaCurve {keys [<at 0 0 RCIM_Cubic>,<at 0.1258 0.9939>,<at 0.5157 1>,<at 1 0 RCIM_Cubic>]}
        BlueCurve {keys [
          {arriveTangent=-13.0252, interpMode=RCIM_Cubic, time=0.0487, value=1}
          <at 0.1695 1.5>
          <at 0.363 1>
          <at 0.4343 0.3077>
          <at 0.5361 0>
          <at 0.6743 0.0097>
          <at 0.9595 0.2031>
         ]}
        GreenCurve {keys [
          {arriveTangent=1.3261, interpMode=RCIM_Cubic, time=0.0487, value=0.7786}
          <at 0.1695 0.421>
          <at 0.363 0.571>
          <at 0.4343 0.6146>
          <at 0.5361 0.4375>
          <at 0.6743 0.4452>
          <at 0.9595 0.1733>
         ]}
        RedCurve {keys [
          {arriveTangent=0.4147, interpMode=RCIM_Cubic, time=0.0487, value=0}
          <at 0.1695 0>
          <at 0.363 0>
          <at 0.4343 0>
          <at 0.5361 0.0704>
          <at 0.6743 0.0685>
          <at 0.9595 0>
         ]}
       }})
     Scale Factor=&MakeCustomVector4FromBool(Bool:NiagaraBool=$User.IsFrontEnd, True Vector4:Vector4f=<0.6 0.6 0.6 1>, False Vector4:Vector4f=<1 1 1 1>)
    )) {isSetParameters=true}
   /SetVariables(Particles.PivotOffset2:Vector2f=&Add_Vector2(
     A:Vector2f=$Particles.PivotOffset
     B:Vector2f=&MakeVector2D(
      X=&FloatFromCurve(
       FloatCurve=DataInterface() {propertyValues {Curve {
          defaultValue=1
          keys [
           {interpMode=RCIM_Cubic, leaveTangentWeight=0.5545, tangentMode=RCTM_User, time=0, value=0}
           {interpMode=RCIM_Cubic, tangentMode=RCTM_User, time=1, value=0.6}
          ]
         }}}
       Scale Curve=&SelectFloatFromArray(Direct Array Index=$Particles.PartitionParticles.Partition, Float Selection Array=DataInterface() {propertyValues {FloatData [-1,1,-1]}})
      )
      Y=&FloatFromCurve(
       FloatCurve=DataInterface() {propertyValues {Curve {
          defaultValue=1
          keys [
           {interpMode=RCIM_Cubic, leaveTangentWeight=0.4502, tangentMode=RCTM_User, time=0, value=0}
           {arriveTangentWeight=0.3333, interpMode=RCIM_Cubic, tangentMode=RCTM_User, time=1, value=0.6}
          ]
         }}}
       Scale Curve=&SelectFloatFromArray(Direct Array Index=$Particles.PartitionParticles.Partition, Float Selection Array=DataInterface() {propertyValues {FloatData [-1,1,-1]}})
      )
     )
    )) {isSetParameters=true}
   /SetVariables(Particles.Position=&AddVectorToPosition(
     Position=$StackContext.InterpolateOverTime.InterpolateOverTime_Position.Moving Average
     Vector=&Multiply_VectorByFloat(
      Vector=&MakeCustomVectorFromBool(
       Bool=$User.IsFrontEnd
       True Vector=&RandomRangeVector(Minimum=<0 0 1>, Maximum=<0 0 1.5>)
       False Vector=<0 0 1>
      )
      Float=&MakeCustomFloatFromBool(
       Bool=$User.IsLargeBody
       True Float=&RandomRangeFloat(Minimum=1.25, Maximum=1.5)
       False Float=&RandomRangeFloat(Minimum=0.5, Maximum=1.25)
      )
     )
    )) {isSetParameters=true}
  ]
  rendererClasses [&NiagaraSpriteRendererProperties]
  renderers [
   {rendererClass=&NiagaraSpriteRendererProperties}
   {rendererClass=&NiagaraSpriteRendererProperties, rendererIndex=1}
  ]
 }
 Speed_Lines {
  emitterUpdateScript [
   &EmitterState()
   &SpawnRate(SpawnRate=20, Use Spawn Probability=false)
   &SpawnRate(SpawnGroup=1, SpawnRate=10, Use Spawn Probability=false) {module=SpawnRate001}
  ]
  particleSpawnScript [
   &PartitionParticles(Number Of Partitions=&MaskIntBySpawnGroup(Masked Int=3, Passthrough Int=5, Spawn Group Mask=0))
   &InitializeParticle(
    Color=&SelectLinearColorFromArray(
     Array Sampling Mode=::NewEnumerator0
     Direct Array Index=?
     Color Selection Array=DataInterface() {propertyValues {ColorData [<rgba 1.5 0.6821 0.1172 1>, <rgba 1.5 0.2257 0.8263 1>, <rgba 0 1.5 1.2853 1>]}}
     Random Seed=0
    )
    Lifetime=&MaskFloatBySpawnGroup(
     Masked Float=&RandomRangeFloat(Minimum=0.25, Maximum=0.4)
     Passthrough Float=&RandomRangeFloat(Minimum=0.35, Maximum=0.6)
    )
    Position Mode=::NewEnumerator1
    Position Offset=<0 0 -80>
    Sprite Size=&MaskVector2dBySpawnGroup(
     Masked Vector2D=&RandomRangeVector2D(Minimum=<0.5 12>, Maximum=<0.75 20>)
     Passthrough Vector2D=&RandomRangeVector2D(Minimum=<0.5 25>, Maximum=<0.75 40>)
     Spawn Group Mask=0
    )
    Sprite Size Mode=::NewEnumerator4
    UsePositionOffset=false
    Write Material Random=true
   )
   &CameraOffset(Camera Offset Amount=-25)
   &ShapeLocation(
    Apply To Particle Position=::NewEnumerator1
    Box Size=<20 2 3>
    MaskSpawnGroup=false
    Override Seed=false
    Shape Origin=&VectorFromFloat(Value=0)
    Shape Primitive=::NewEnumerator2
    Surface Only=false
    UseOwnerScale=true
   )
   &ShapeLocation(
    Apply To Particle Position=::NewEnumerator1
    MaskSpawnGroup=false
    Override Seed=false
    Shape Origin=&VectorFromFloat(Value=0)
    Sphere Radius=10
    UseOwnerScale=true
   ) {module=ShapeLocation001}
   &AddVelocity(Velocity=&MaskVectorBySpawnGroup(
     Masked Vector=&RandomRangeVector(Minimum=<0 0 100>, Maximum=<0 0 150>)
     Passthrough Vector=&RandomRangeVector(Minimum=<0 0 100>, Maximum=<0 0 300>)
    ))
  ]
  particleUpdateScript [
   &ParticleState()
   &ScaleSpriteSize(
    Non-Uniform Curve Sprite Scale=DataInterface() {propertyValues {
      XCurve {keys [<at 0.7 1 RCIM_Cubic>, {arriveTangent=-1.0042, interpMode=RCIM_Cubic, leaveTangent=-1.0042, tangentMode=RCTM_User, time=1, value=0.8625}]}
      YCurve {keys [<at 0.7 1 RCIM_Cubic>, {arriveTangent=0.6449, interpMode=RCIM_Cubic, leaveTangent=0.6449, tangentMode=RCTM_User, time=1, value=1.0875}]}
     }}
    Scale Sprite Size Mode=::NewEnumerator3
   )
   &SkeletalMeshLocation(
    Bone / Socket Index=&SelectIntFromArray(Direct Array Index=$Particles.PartitionParticles.Partition, Int Selection Array=DataInterface() {propertyValues {IntData [0,1,2,3,4]}})
    Bone Sampling Mode=::NewEnumerator4
    MaskSpawnGroup=false
    Orientation Sampling=::NewEnumerator1
    Position Sampling=::NewEnumerator1
    Sampled Position Offset=&MaskVectorBySpawnGroup(
     Masked Vector=Linked() {linkedVariable {name=StackContext.ShapeLocation.ShapePosition, type=NiagaraPosition}}
     Passthrough Vector=Linked() {linkedVariable {name=StackContext.ShapeLocation001.ShapePosition, type=NiagaraPosition}}
     Spawn Group Mask=1
    )
    Sampled Position Offset Coordinate Space:ENiagaraSkelMeshTransforms=::NewEnumerator2
    Skeletal Mesh=$System.Skeletal Mesh
   ) {module=SkeletalMeshLocation003}
   &ScaleColor(
    Linear Color Curve=DataInterface() {propertyValues {
      AlphaCurve {keys [<at 0 0>,<at 0.1007 0.8>,<at 0.3932 0.2>,<at 1 0>]}
      BlueCurve {keys [<at 0 1>,<at 0.8464 0.4485>]}
      GreenCurve {keys [<at 0 1>,<at 0.8464 0.1254>]}
      RedCurve {keys [<at 0 1>,<at 0.8464 1>]}
     }}
    Scale Mode=::NewEnumerator2
   )
   &Drag(
    Drag=&MaskFloatBySpawnGroup(Masked Float=7, Passthrough Float=3)
    Use Linear Drag=true
    Use Rotational Drag=true
   )
   &SolveForcesAndVelocity(Clamp Velocity=false, Limit Acceleration=false)
   &ScaleSpriteSizeBySpeed(Max Scale Factor=<1.2 3>, Min Scale Factor=<1 1.4>) {enabled=false}
  ]
  rendererClasses [&NiagaraSpriteRendererProperties]
  renderers [{rendererClass=&NiagaraSpriteRendererProperties}]
 }
 Speed_Lines001 {
  bEnabled=false
  emitterUpdateScript [
   &EmitterState()
   &SpawnRate(SpawnRate=30, Use Spawn Probability=false)
   &SpawnRate(SpawnGroup=1, SpawnRate=15, Use Spawn Probability=false) {module=SpawnRate001}
  ]
  particleSpawnScript [
   &PartitionParticles(Number Of Partitions=&MaskIntBySpawnGroup(Masked Int=3, Passthrough Int=5, Spawn Group Mask=0))
   &InitializeParticle(
    Color=&SelectLinearColorFromArray(
     Array Sampling Mode=::NewEnumerator0
     Direct Array Index=?
     Color Selection Array=DataInterface() {propertyValues {ColorData [<rgba 2 0.9095 0.1562 1>, <rgba 2 0.3009 1.1018 1>, <rgba 0 2 1.7137 1>]}}
     Random Seed=0
    )
    Lifetime=&MaskFloatBySpawnGroup(
     Masked Float=&RandomRangeFloat(Minimum=0.2, Maximum=0.6)
     Passthrough Float=&FloatFromCurve(
      FloatCurve=DataInterface() {propertyValues {Curve {keys [<at 0.2 0.2>,<at 0.8 0.25>]}}}
      CurveIndex=&RandomRangeFloat(Minimum=0, Maximum=1)
     )
    )
    Position Mode=::NewEnumerator1
    Position Offset=<0 0 -80>
    Sprite Size=&MaskVector2dBySpawnGroup(
     Masked Vector2D=&RandomRangeVector2D(Minimum=<0.5 9>, Maximum=<1 20>)
     Passthrough Vector2D=&RandomRangeVector2D(Minimum=<0.3 10>, Maximum=<1.2 40>)
     Spawn Group Mask=0
    )
    Sprite Size Mode=::NewEnumerator4
    UsePositionOffset=false
    Write Material Random=true
   )
   &CameraOffset(Camera Offset Amount=-25)
   &ShapeLocation(
    Apply To Particle Position=::NewEnumerator1
    Box Size=&MaskVectorBySpawnGroup(Masked Vector=<25 2 3>, Passthrough Vector=<6 6 6>, Spawn Group Mask=1)
    MaskSpawnGroup=false
    Override Seed=false
    Shape Origin=&VectorFromFloat(Value=0)
    Shape Primitive=::NewEnumerator2
    Surface Only=false
    UseOwnerScale=true
   )
   &ShapeLocation(
    Apply To Particle Position=::NewEnumerator1
    MaskSpawnGroup=false
    Override Seed=false
    Shape Origin=&VectorFromFloat(Value=0)
    Sphere Radius=6
    Sphere Surface Distribution=0.6667
    UseOwnerScale=true
   ) {module=ShapeLocation001}
   &AddVelocity(Velocity=&MaskVectorBySpawnGroup(
     Masked Vector=&RandomRangeVector(Minimum=<0 0 100>, Maximum=<0 0 200>)
     Passthrough Vector=&RandomRangeVector(Minimum=<0 0 100>, Maximum=<0 0 300>)
    ))
  ]
  particleUpdateScript [
   &ParticleState()
   &SkeletalMeshLocation(
    Bone / Socket Index=&SelectIntFromArray(Direct Array Index=$Particles.PartitionParticles.Partition, Int Selection Array=DataInterface() {propertyValues {IntData [0,1,2,3,4]}})
    Bone Sampling Mode=::NewEnumerator4
    MaskSpawnGroup=false
    Orientation Sampling=::NewEnumerator1
    Position Sampling=::NewEnumerator1
    Sampled Position Offset=&MaskVectorBySpawnGroup(
     Masked Vector=Linked() {linkedVariable {name=StackContext.ShapeLocation.ShapePosition, type=NiagaraPosition}}
     Passthrough Vector=Linked() {linkedVariable {name=StackContext.ShapeLocation001.ShapePosition, type=NiagaraPosition}}
     Spawn Group Mask=1
    )
    Sampled Position Offset Coordinate Space:ENiagaraSkelMeshTransforms=::NewEnumerator2
    Skeletal Mesh=$System.Skeletal Mesh
   ) {module=SkeletalMeshLocation003}
   &ScaleColor(
    Linear Color Curve=DataInterface() {propertyValues {
      AlphaCurve {keys [<at 0 0>,<at 0.1145 1>,<at 0.5676 0.3>,<at 1 0>]}
      BlueCurve {keys [<at 0 1>,<at 0.8464 0.4485>]}
      GreenCurve {keys [<at 0 1>,<at 0.8464 0.1254>]}
      RedCurve {keys [<at 0 1>,<at 0.8464 1>]}
     }}
    Scale Mode=::NewEnumerator2
   )
   &Drag(Drag=3, Use Linear Drag=true, Use Rotational Drag=true)
   &SolveForcesAndVelocity(Clamp Velocity=false, Limit Acceleration=false)
   &ScaleSpriteSizeBySpeed(Max Scale Factor=<1.2 3>, Min Scale Factor=<1 1.4>)
  ]
  rendererClasses [&NiagaraSpriteRendererProperties]
  renderers [{rendererClass=&NiagaraSpriteRendererProperties}]
 }
}
summary {
 systemName=NS_Aura_Pixel
 emitters [
  {emitterName=Glow_AndDecals, rendererClasses [&NiagaraSpriteRendererProperties]}
  {emitterName=Pixels_Animated_Floating, rendererClasses [&NiagaraSpriteRendererProperties]}
  {emitterName=Pixel_small_Particles, rendererClasses [&NiagaraSpriteRendererProperties]}
  {emitterName=Bright_Glow_backer, rendererClasses [&NiagaraSpriteRendererProperties]}
  {emitterName=Speed_Lines, rendererClasses [&NiagaraSpriteRendererProperties]}
  {emitterName=Pixels_onBody, rendererClasses [&NiagaraSpriteRendererProperties]}
  {bEnabled=false, emitterName=Speed_Lines001, rendererClasses [&NiagaraSpriteRendererProperties]}
 ]
 userVariables [
  User.IsFrontEnd {type=NiagaraBool}
  User.IsLargeBody {type=NiagaraBool}
 ]
}
