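# Niagara sprite system for the start of the Pixel aura cosmetic, anchored to a skeletal mesh.
# A brief flare, corona and glow accompany one-shot bursts of multicolored pixels and speed lines; a duplicate pixel burst is disabled.
# Rate-spawned pixels rise from partitioned bones or sockets, using smoothed positions plus stepped-curve offsets for blocky motion.
# User.IsFrontEnd dims secondary pixel color, and User.IsLargeBody enlarges and speeds up the floating pixels.
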NiagaraSystem /SparksCosmetics/Items/Auras/Assets/NS_Aura_Pixel/NS_Aura_Pixel_Start
dependencies {
 usedDataInterfaces [
  &NiagaraDataInterfaceCurve
  &NiagaraDataInterfaceColorCurve
  &NiagaraDataInterfaceArrayColor
  &NiagaraDataInterfaceArrayFloat
  &NiagaraDataInterfaceArrayInt32
  &NiagaraDataInterfaceVector2DCurve
  &NiagaraDataInterfaceSkeletalMesh
 ]
 usedDynamicInputs [&MakeCustomLinearColorFromBool, &RandomBool]
 usedModules [
  &self:SystemSpawnScript.SystemScriptSource.SystemScriptGraph.NiagaraNodeAssignment_0.SetVariables_#1
  &SystemState
 ]
}
emitters {
 Corona {
  emitterUpdateScript [
   &EmitterState(Life Cycle Mode=::NewEnumerator1, Loop Behavior=::NewEnumerator1, Loop Duration=1, UseLoopDelay=false)
   &SpawnBurst_Instantaneous(Spawn Count=1, Use Spawn Probability=false)
  ]
  particleSpawnScript [
   &InitializeParticle(
    Color=<rgba 0.6924 0.2297 0.1463 1>
    Lifetime=0.25
    Sprite Rotation Mode=::NewEnumerator2
    Sprite Size Mode=::NewEnumerator3
    Uniform Sprite Size=400
    UsePositionOffset=false
    Write Material Random=true
   )
   &DynamicMaterialParameters(
    Index 0 Param 1=&RandomRangeFloat(Minimum=0.4, Maximum=0.6)
    Index 0 Param 2=1
    Index 0 Param 3=1
    Index 0 Param 4=1
    Module.Param0WriteEnabled=true
    Module.Param1WriteEnabled=true
    Module.Param2WriteEnabled=true
    Module.Param3WriteEnabled=true
   )
   &SystemLocation(MaskSpawnGroup=false, Offset=<0 0 120>)
  ]
  particleUpdateScript [
   &ParticleState()
   &SolveForcesAndVelocity(Clamp Velocity=false, Limit Acceleration=false)
   &SkeletalMeshLocation(Bone Sampling Mode=::NewEnumerator4, MaskSpawnGroup=false, Skeletal Mesh=$System.Skeletal Mesh) {module=SkeletalMeshLocation001}
   &ScaleColor(
    Linear Color Curve=DataInterface() {propertyValues {
      AlphaCurve {keys [
        {arriveTangent=6.6262, interpMode=RCIM_Cubic, leaveTangent=6.6262, tangentMode=RCTM_User, time=0, value=0}
        <at 0.2 0.65 RCIM_Cubic>
        <at 1 0 RCIM_Cubic>
       ]}
      BlueCurve {keys [<at 0 1>,<at 1 0.0582>]}
      GreenCurve {keys [<at 0 1>,<at 1 0.0929>]}
      RedCurve {keys [<at 0 1>,<at 1 0.3021>]}
     }}
    Scale Mode=::NewEnumerator2
   )
   &ScaleSpriteSize(Scale Factor=&Vector2DFromFloat(Value=&FloatFromCurve(FloatCurve=DataInterface() {propertyValues {Curve {keys [
          {arriveTangent=2.1722, interpMode=RCIM_Cubic, leaveTangent=2.1722, tangentMode=RCTM_User, time=0, value=0}
          {arriveTangent=0.0224, interpMode=RCIM_Cubic, leaveTangent=0.0224, tangentMode=RCTM_User, time=1, value=1}
         ]}}})))
   &CameraOffset(Camera Offset Amount=-30)
   &DynamicMaterialParameters(
    Index 0 Param 1=0.75
    Index 0 Param 2=2
    Index 0 Param 3=1
    Module.Param0WriteEnabled=true
    Module.Param1WriteEnabled=true
    Module.Param2WriteEnabled=true
    Module.Param3WriteEnabled=true
   ) {module=DynamicMaterialParameters001}
  ]
  rendererClasses [&NiagaraSpriteRendererProperties]
  renderers [{rendererClass=&NiagaraSpriteRendererProperties}]
 }
 Dark_Glow {
  emitterUpdateScript [
   &EmitterState(
    Inactive Response=::NewEnumerator1
    Life Cycle Mode=::NewEnumerator1
    Loop Behavior=::NewEnumerator1
    Loop Delay=0.1
    Loop Duration=1
    UseLoopDelay=false
   )
   &SpawnBurst_Instantaneous(Spawn Count=1, Use Spawn Probability=false)
  ]
  particleSpawnScript [&InitializeParticle(
    Color=<rgba 0.387 0.0914 0.0276 1>
    Lifetime=0.25
    Position Offset=<0 0 100>
    Sprite Rotation Mode=::NewEnumerator2
    Sprite Size Mode=::NewEnumerator3
    Uniform Sprite Size=230
    UsePositionOffset=true
    Write Material Random=true
   )]
  particleUpdateScript [
   &ParticleState()
   &SkeletalMeshLocation(Bone Sampling Mode=::NewEnumerator4, MaskSpawnGroup=false, Skeletal Mesh=$System.Skeletal Mesh) {module=SkeletalMeshLocation001}
   &ScaleColor(
    Scale Alpha=&FloatFromCurve(
     FloatCurve=DataInterface() {propertyValues {Curve {keys [
         {interpMode=RCIM_Cubic, tangentMode=RCTM_User, time=0, value=0}
         {interpMode=RCIM_Cubic, tangentMode=RCTM_User, time=0.2, value=1}
         {interpMode=RCIM_Cubic, tangentMode=RCTM_User, time=0.5, value=1}
         {tangentMode=RCTM_User, time=1, value=0}
        ]}}}
     Scale Curve=2
    )
    ScaleA=true
    ScaleRGB=true
   )
  ]
  rendererClasses [&NiagaraSpriteRendererProperties]
  renderers [{rendererClass=&NiagaraSpriteRendererProperties}]
 }
 Flare {
  emitterUpdateScript [
   &EmitterState(Life Cycle Mode=::NewEnumerator1, Loop Behavior=::NewEnumerator1, Loop Duration=1, UseLoopDelay=false)
   &SpawnBurst_Instantaneous(Spawn Count=1, Use Spawn Probability=false)
  ]
  particleSpawnScript [
   &InitializeParticle(
    Color=<rgba 1.25 0.4726 0.3288 1>
    Lifetime=0.25
    Sprite Rotation Mode=::NewEnumerator2
    Sprite Size Mode=::NewEnumerator3
    Uniform Sprite Size=600
    UsePositionOffset=false
    Write Material Random=true
   )
   &DynamicMaterialParameters(
    Index 0 Param 1=&RandomRangeFloat(Minimum=0.4, Maximum=0.6)
    Index 0 Param 2=1
    Index 0 Param 3=1
    Index 0 Param 4=1
    Module.Param0WriteEnabled=true
    Module.Param1WriteEnabled=true
    Module.Param2WriteEnabled=true
    Module.Param3WriteEnabled=true
   )
   &SystemLocation(MaskSpawnGroup=false, Offset=<0 0 120>)
   /SetVariables(Particles.Sharp Flare Camera Offset=40) {isSetParameters=true}
   &CameraOffset(Camera Offset Amount=-30) {module=CameraOffset001}
  ]
  particleUpdateScript [
   &ParticleState()
   &SolveForcesAndVelocity(Clamp Velocity=false, Limit Acceleration=false)
   &SkeletalMeshLocation(Bone Sampling Mode=::NewEnumerator4, MaskSpawnGroup=false, Skeletal Mesh=$System.Skeletal Mesh) {module=SkeletalMeshLocation001}
   &ScaleColor(
    Scale Alpha=&Multiply_Float(
     A=&FNTimeOfDayScalar(Night Scale=1, Morning Scale=1, Evening Scale=1)
     B=&FloatFromCurve(FloatCurve=DataInterface() {propertyValues {Curve {keys [<at 0 0>,<at 0.2 1>,<at 0.8 1>,<at 1 0>]}}})
    )
    Scale RGB=&Multiply_VectorByFloat(Vector=<1 1 1>, Float=1)
    ScaleA=true
    ScaleRGB=true
   )
   &ScaleSpriteSize(Scale Factor=&Vector2DFromFloat(Value=&FloatFromCurve(FloatCurve=DataInterface() {propertyValues {Curve {keys [<at 0 1.5 RCIM_Cubic>,<at 1 0>]}}})))
   /SetVariables(Particles.Flash color:LinearColor=&MakeLinearColorFromVectorAndFloat(
     "Vector (RGB)"=&MakeVectorFromLinearColorRGB(Color=<rgba 1 0.3452 0.2256 1>)
     "Float (Alpha)"=&Multiply_Float(
      A=&Multiply_Float(
       A=&FNTimeOfDayScalar(Night Scale=1, Morning Scale=1, Evening Scale=1)
       B=&FloatFromCurve(FloatCurve=DataInterface() {propertyValues {Curve {keys [<at 0 0>,<at 0.124 1>,<at 0.496 1>,<at 0.62 0>]}}})
      )
      B=1
     )
    )) {isSetParameters=true}
   /SetVariables(Particles.Flash size:Vector2f=&Multiply_Vector2D_ByFloat(
     Float=&Multiply_Float(
      A=400
      B=&FloatFromCurve(FloatCurve=DataInterface() {propertyValues {Curve {keys [
           {arriveTangent=4.1785, interpMode=RCIM_Cubic, leaveTangent=4.1785, tangentMode=RCTM_User, time=0, value=0}
           {arriveTangent=-0.15, interpMode=RCIM_Cubic, leaveTangent=-0.15, tangentMode=RCTM_User, time=0.1378, value=1}
           {arriveTangent=-0.0405, interpMode=RCIM_Cubic, leaveTangent=-0.0405, tangentMode=RCTM_User, time=0.62, value=0.25}
          ]}}})
     )
     Vector2D=<1 1>
    )) {isSetParameters=true}
  ]
  rendererClasses [&NiagaraSpriteRendererProperties]
  renderers [
   {rendererClass=&NiagaraSpriteRendererProperties}
   {rendererClass=&NiagaraSpriteRendererProperties, rendererIndex=1}
  ]
 }
 Pixel_Burst_Animated {
  emitterUpdateScript [
   &EmitterState()
   &SpawnBurst_Instantaneous(Spawn Count=30, Use Spawn Probability=false, UseLoopCountLimit=true)
   &SpawnRate(SpawnRate=&RandomRangeFloat(Minimum=4, Maximum=8), Use Spawn Probability=false) {enabled=false}
  ]
  particleSpawnScript [
   &InitializeParticle(
    Lifetime Max=0.8
    Lifetime Min=0.4
    Lifetime Mode=::NewEnumerator1
    Sprite Size Mode=::NewEnumerator1
    Uniform Sprite Size Max=7
    Uniform Sprite Size Min=1
    UsePositionOffset=false
    Write Material Random=true
   )
   /SetVariables() {isSetParameters=true, module=SetVariables_#2}
   &CameraOffset(Camera Offset Amount=-25)
   &ShapeLocation(MaskSpawnGroup=false, Offset Mode=::NewEnumerator0, Override Seed=false, Sphere Radius=20, UseOwnerScale=true)
   &AddVelocity(
    Module.Use Falloff Distance=false
    Override Seed=false
    Velocity Mode=::NewEnumerator1
    Velocity Speed=&RandomRangeFloat(Minimum=80, Maximum=450)
   ) {module=AddVelocity001}
   &AddVelocity(
    Velocity=&MakeVector(
     X=&RandomRangeFloat(Minimum=-400, Maximum=400)
     Y=&RandomRangeFloat(Minimum=-400, Maximum=400)
     Z=&RandomRangeFloat(Minimum=-400, Maximum=400)
    )
    Velocity Speed Scale=&RandomRangeFloat(Minimum=0.2, Maximum=1)
   ) {enabled=false}
  ]
  particleUpdateScript [
   &ParticleState()
   &Drag(Drag=4, Use Linear Drag=true, Use Rotational Drag=true)
   &ScaleSpriteSize(
    Uniform Curve Scale=0.6
    Uniform Curve Sprite Scale=DataInterface() {propertyValues {Curve {keys [
        {arriveTangent=8.9937, interpMode=RCIM_Cubic, leaveTangent=8.9937, tangentMode=RCTM_User, time=0, value=0}
        {arriveTangentWeight=0.1601, interpMode=RCIM_Cubic, leaveTangentWeight=0.2668, tangentMode=RCTM_User, time=0.2, value=1}
        {arriveTangentWeight=0.2906, interpMode=RCIM_Cubic, leaveTangentWeight=0.1163, tangentMode=RCTM_User, time=0.7, value=1}
        {arriveTangent=-6.7083, interpMode=RCIM_Cubic, leaveTangent=-6.7083, tangentMode=RCTM_User, time=1, value=0}
       ]}}}
   ) {enabled=false}
   &ScaleSpriteSize(Uniform Curve Sprite Scale=DataInterface() {propertyValues {Curve {
       defaultValue=0.85
       keys [
        {arriveTangent=6.7723, interpMode=RCIM_Cubic, leaveTangent=6.7723, tangentMode=RCTM_User, time=0, value=0.1}
        {arriveTangent=-0.0348, interpMode=RCIM_Cubic, leaveTangent=-0.0348, tangentMode=RCTM_User, time=0.3, value=1}
        {arriveTangent=1.2587, arriveTangentWeight=0.2667, interpMode=RCIM_Cubic, leaveTangent=1.2587, tangentMode=RCTM_User, time=1, value=1.3}
       ]
      }}}) {module=ScaleSpriteSize001}
   &SolveForcesAndVelocity(Clamp Velocity=false, Limit Acceleration=false)
   &SkeletalMeshLocation(
    Bone / Socket Index=$Particles.SpawnGroup
    Bone Sampling Mode=::NewEnumerator4
    MaskSpawnGroup=false
    Position Sampling=::NewEnumerator1
    Sampled Position Offset=$StackContext.ShapeLocation.ShapePosition
    Skeletal Mesh=$System.Skeletal Mesh
    Source Movement Scale:Vector3f=<0 0 0>
   ) {module=SkeletalMeshLocation001}
   /SetVariables(Particles.color2:LinearColor=&MultiplyLinearColor(
     Linear Color=&ColorFromCurve(ColorCurve=DataInterface() {propertyValues {
        AlphaCurve {keys [<at 0 0 RCIM_Cubic>,<at 0.0386 0.9939>,<at 0.6985 1>,<at 1 0 RCIM_Cubic>]}
        BlueCurve {keys [
          <at 0.0588 0.3116>
          <at 0.1258 0.7292>
          <at 0.204 1>
          {arriveTangent=-1.209, interpMode=RCIM_Cubic, leaveTangent=-1.209, time=0.413, value=0.8958}
          <at 0.5419 0>
          <at 0.9905 0.0492>
         ]}
        GreenCurve {keys [
          <at 0.0588 0.6823>
          <at 0.1258 0>
          <at 0.204 0.0972>
          {arriveTangent=0.2803, interpMode=RCIM_Cubic, leaveTangent=0.2803, time=0.413, value=0.6975}
          <at 0.5419 0.5052>
          <at 0.9905 0.3177>
         ]}
        RedCurve {keys [<at 0.0588 0>,<at 0.1258 0.5831>,<at 0.204 0>,<at 0.413 0 RCIM_Cubic>,<at 0.5419 0.0813>,<at 0.9905 0>]}
       }})
     Scale Factor=/Niagara/DynamicInputs/TypeConversions/MakeCustomVector4FromBool(Bool:NiagaraBool=$User.IsFrontEnd, True Vector4:Vector4f=<0.6 0.6 0.6 1>, False Vector4:Vector4f=<1 1 1 1>)
    )) {isSetParameters=true}
   &ScaleColor(
    Linear Color Curve=DataInterface() {propertyValues {
      AlphaCurve {keys [<at 0 0>,<at 0.0324 1>,<at 0.5888 1>,<at 1 0>]}
      BlueCurve {keys [<at 0.0382 1>,<at 0.1842 0>,<at 0.5818 0.0173>,<at 0.8528 0.2174>]}
      GreenCurve {keys [<at 0.0382 0.0007>,<at 0.1842 0.0969>,<at 0.5818 0>,<at 0.8528 0>]}
      RedCurve {keys [<at 0.0382 0.5893>,<at 0.1842 1.5>,<at 0.5818 1.1>,<at 0.8528 0.5885>]}
     }}
    Scale Mode=::NewEnumerator2
   )
   /SetVariables(Particles.PivotOffset2:Vector2f=/Niagara/DynamicInputs/Add/Add_Vector2(
     A:Vector2f=$Particles.PivotOffset
     B:Vector2f=&MakeVector2D(
      X=&FloatFromCurve(FloatCurve=DataInterface() {propertyValues {Curve {
          defaultValue=1
          keys [{interpMode=RCIM_Cubic, tangentMode=RCTM_User, time=0, value=0}, {interpMode=RCIM_Cubic, tangentMode=RCTM_User, time=1, value=0.3}]
         }}})
      Y=&FloatFromCurve(FloatCurve=DataInterface() {propertyValues {Curve {
          defaultValue=1
          keys [
           {interpMode=RCIM_Cubic, tangentMode=RCTM_User, time=0, value=0}
           {arriveTangentWeight=0.3333, interpMode=RCIM_Cubic, tangentMode=RCTM_User, time=1, value=0.3}
          ]
         }}})
     )
    )) {isSetParameters=true, module=SetVariables_#3}
   &DynamicMaterialParameters(
    Index 0 Param 1=&FloatFromCurve(
     FloatCurve=DataInterface() {propertyValues {Curve {defaultValue=1, keys [{tangentMode=RCTM_User, time=0.7, value=0}, {tangentMode=RCTM_User, time=1, value=1}]}}}
     Scale Curve=&SelectFloatFromArray(
      Array Sampling Mode=::NewEnumerator0
      Direct Array Index=?
      Float Selection Array=DataInterface() {propertyValues {FloatData [0,2,2]}}
      Random Seed=0
     )
    )
    Index 0 Param 2=&FloatFromCurve(FloatCurve=DataInterface() {propertyValues {Curve {defaultValue=1, keys [{tangentMode=RCTM_User, time=0.4, value=0}, {tangentMode=RCTM_User, time=0.7, value=3}]}}})
    Module.Index0Float0Write=true
    Module.Index0Float1Write=true
    Module.Index0Float2Write=true
    Module.Index0Float3Write=true
   )
  ]
  rendererClasses [&NiagaraSpriteRendererProperties]
  renderers [
   {rendererClass=&NiagaraSpriteRendererProperties}
   {rendererClass=&NiagaraSpriteRendererProperties, rendererIndex=1}
  ]
 }
 Pixel_Burst_Animated001 {
  bEnabled=false
  emitterUpdateScript [
   &EmitterState()
   &SpawnBurst_Instantaneous(Spawn Count=30, Use Spawn Probability=false, UseLoopCountLimit=true)
   &SpawnRate(SpawnRate=&RandomRangeFloat(Minimum=4, Maximum=8), Use Spawn Probability=false) {enabled=false}
  ]
  particleSpawnScript [
   &InitializeParticle(
    Lifetime Max=0.8
    Lifetime Min=0.4
    Lifetime Mode=::NewEnumerator1
    Sprite Size Mode=::NewEnumerator1
    Uniform Sprite Size Max=7
    Uniform Sprite Size Min=1
    UsePositionOffset=false
    Write Material Random=true
   )
   /SetVariables() {isSetParameters=true, module=SetVariables_#2}
   &CameraOffset(Camera Offset Amount=-25)
   &ShapeLocation(MaskSpawnGroup=false, Offset Mode=::NewEnumerator0, Override Seed=false, Sphere Radius=20, UseOwnerScale=true)
   &AddVelocity(
    Module.Use Falloff Distance=false
    Override Seed=false
    Velocity Mode=::NewEnumerator1
    Velocity Speed=&RandomRangeFloat(Minimum=80, Maximum=450)
   ) {module=AddVelocity001}
   &AddVelocity(
    Velocity=&MakeVector(
     X=&RandomRangeFloat(Minimum=-400, Maximum=400)
     Y=&RandomRangeFloat(Minimum=-400, Maximum=400)
     Z=&RandomRangeFloat(Minimum=-400, Maximum=400)
    )
    Velocity Speed Scale=&RandomRangeFloat(Minimum=0.2, Maximum=1)
   ) {enabled=false}
  ]
  particleUpdateScript [
   &ParticleState()
   &Drag(Drag=4, Use Linear Drag=true, Use Rotational Drag=true)
   &ScaleSpriteSize(
    Uniform Curve Scale=0.6
    Uniform Curve Sprite Scale=DataInterface() {propertyValues {Curve {keys [
        {arriveTangent=8.9937, interpMode=RCIM_Cubic, leaveTangent=8.9937, tangentMode=RCTM_User, time=0, value=0}
        {arriveTangentWeight=0.1601, interpMode=RCIM_Cubic, leaveTangentWeight=0.2668, tangentMode=RCTM_User, time=0.2, value=1}
        {arriveTangentWeight=0.2906, interpMode=RCIM_Cubic, leaveTangentWeight=0.1163, tangentMode=RCTM_User, time=0.7, value=1}
        {arriveTangent=-6.7083, interpMode=RCIM_Cubic, leaveTangent=-6.7083, tangentMode=RCTM_User, time=1, value=0}
       ]}}}
   ) {enabled=false}
   &ScaleSpriteSize(Uniform Curve Sprite Scale=DataInterface() {propertyValues {Curve {
       defaultValue=0.85
       keys [
        {arriveTangent=6.7723, interpMode=RCIM_Cubic, leaveTangent=6.7723, tangentMode=RCTM_User, time=0, value=0.1}
        {arriveTangent=-0.0348, interpMode=RCIM_Cubic, leaveTangent=-0.0348, tangentMode=RCTM_User, time=0.3, value=1}
        {arriveTangent=1.2587, arriveTangentWeight=0.2667, interpMode=RCIM_Cubic, leaveTangent=1.2587, tangentMode=RCTM_User, time=1, value=1.3}
       ]
      }}}) {module=ScaleSpriteSize001}
   &SolveForcesAndVelocity(Clamp Velocity=false, Limit Acceleration=false)
   &SkeletalMeshLocation(
    Bone / Socket Index=$Particles.SpawnGroup
    Bone Sampling Mode=::NewEnumerator4
    MaskSpawnGroup=false
    Position Sampling=::NewEnumerator1
    Sampled Position Offset=$StackContext.ShapeLocation.ShapePosition
    Skeletal Mesh=$System.Skeletal Mesh
    Source Movement Scale:Vector3f=<0 0 0>
   ) {module=SkeletalMeshLocation001}
   &ScaleColor(
    Linear Color Curve=DataInterface() {propertyValues {
      AlphaCurve {keys [<at 0 0>,<at 0.1325 1>,<at 0.4026 1>,<at 1 0>]}
      BlueCurve {keys [<at 0.0168 0.4714>,<at 0.1741 0>,<at 1 0.1145>]}
      GreenCurve {keys [<at 0.0168 0.0003>,<at 0.1741 0.0359>,<at 1 0>]}
      RedCurve {keys [<at 0.0168 0.2778>,<at 0.1741 0.556>,<at 1 0.3099>]}
     }}
    Scale Mode=::NewEnumerator2
   )
   /SetVariables(Particles.color2:LinearColor=&ColorFromCurve(ColorCurve=DataInterface() {propertyValues {
       AlphaCurve {keys [<at 0 0 RCIM_Cubic>,<at 0.2 0.9939>,<at 0.4363 1>,<at 1 0 RCIM_Cubic>]}
       BlueCurve {keys [
         <at 0.0437 2>
         {arriveTangent=-6.2302, interpMode=RCIM_Cubic, leaveTangent=-6.2302, time=0.1044, value=1}
         <at 0.3647 0>
         <at 0.6249 0.191>
         <at 1 0>
        ]}
       GreenCurve {keys [
         <at 0.0437 0.1922>
         {arriveTangent=1.1604, interpMode=RCIM_Cubic, leaveTangent=1.1604, time=0.1044, value=0.7786}
         <at 0.3647 0.5647>
         <at 0.6249 0.3814>
         <at 1 0.1172>
        ]}
       RedCurve {keys [
         <at 0.0437 0>
         {arriveTangent=0.283, interpMode=RCIM_Cubic, leaveTangent=0.283, time=0.1044, value=0}
         <at 0.3647 0.0908>
         <at 0.6249 0>
         <at 1 0.0189>
        ]}
      }})) {isSetParameters=true}
   /SetVariables(Particles.PivotOffset2:Vector2f=&Add_Vector2(
     A:Vector2f=$Particles.PivotOffset
     B:Vector2f=&MakeVector2D(
      X=&FloatFromCurve(FloatCurve=DataInterface() {propertyValues {Curve {
          defaultValue=1
          keys [{interpMode=RCIM_Cubic, tangentMode=RCTM_User, time=0, value=0}, {interpMode=RCIM_Cubic, tangentMode=RCTM_User, time=1, value=0.3}]
         }}})
      Y=&FloatFromCurve(FloatCurve=DataInterface() {propertyValues {Curve {
          defaultValue=1
          keys [
           {interpMode=RCIM_Cubic, tangentMode=RCTM_User, time=0, value=0}
           {arriveTangentWeight=0.3333, interpMode=RCIM_Cubic, tangentMode=RCTM_User, time=1, value=0.3}
          ]
         }}})
     )
    )) {isSetParameters=true, module=SetVariables_#3}
   &DynamicMaterialParameters(
    Index 0 Param 1=&FloatFromCurve(
     FloatCurve=DataInterface() {propertyValues {Curve {defaultValue=1, keys [{tangentMode=RCTM_User, time=0.7, value=0}, {tangentMode=RCTM_User, time=1, value=1}]}}}
     Scale Curve=&SelectFloatFromArray(
      Array Sampling Mode=::NewEnumerator0
      Direct Array Index=?
      Float Selection Array=DataInterface() {propertyValues {FloatData [0,2,2]}}
      Random Seed=0
     )
    )
    Index 0 Param 2=&FloatFromCurve(FloatCurve=DataInterface() {propertyValues {Curve {defaultValue=1, keys [{tangentMode=RCTM_User, time=0.4, value=0}, {tangentMode=RCTM_User, time=0.7, value=3}]}}})
    Module.Index0Float0Write=true
    Module.Index0Float1Write=true
    Module.Index0Float2Write=true
    Module.Index0Float3Write=true
   )
  ]
  rendererClasses [&NiagaraSpriteRendererProperties]
  renderers [
   {rendererClass=&NiagaraSpriteRendererProperties}
   {rendererClass=&NiagaraSpriteRendererProperties, rendererIndex=1}
  ]
 }
 Pixel_small_burst {
  emitterUpdateScript [
   &EmitterState()
   &SpawnBurst_Instantaneous(Spawn Count=20, Use Spawn Probability=false, UseLoopCountLimit=true)
   &SpawnRate(SpawnRate=&RandomRangeFloat(Minimum=4, Maximum=8), Use Spawn Probability=false) {enabled=false}
  ]
  particleSpawnScript [
   &InitializeParticle(
    Color=&SelectLinearColorFromArray(
     Array Sampling Mode=::NewEnumerator0
     Direct Array Index=?
     Color Selection Array=DataInterface() {propertyValues {ColorData [<rgba 0.9351 1 0 0.9>, <rgba 0.0189 1 0.9226 0.7>, <rgba 1 0 0.4671 0.8>]}}
     Random Seed=0
    )
    Lifetime Max=0.5
    Lifetime Min=0.2
    Lifetime Mode=::NewEnumerator1
    Sprite Size Mode=::NewEnumerator1
    Uniform Sprite Size Max=12
    Uniform Sprite Size Min=3
    UsePositionOffset=false
    Write Material Random=true
   )
   &ShapeLocation(MaskSpawnGroup=false, Offset Mode=::NewEnumerator0, Override Seed=false, Sphere Radius=10, UseOwnerScale=true)
   &AddVelocity(
    Module.Use Falloff Distance=false
    Override Seed=false
    Velocity Mode=::NewEnumerator1
    Velocity Speed=&RandomRangeFloat(Minimum=300, Maximum=800)
   ) {module=AddVelocity001}
  ]
  particleUpdateScript [
   &ParticleState()
   &CameraOffset(Camera Offset Amount=-25)
   &Drag(Drag=4, Use Linear Drag=true, Use Rotational Drag=true)
   &ScaleSpriteSize(
    Uniform Curve Scale=0.6
    Uniform Curve Sprite Scale=DataInterface() {propertyValues {Curve {keys [
        {arriveTangent=8.9937, interpMode=RCIM_Cubic, leaveTangent=8.9937, tangentMode=RCTM_User, time=0, value=0}
        {arriveTangentWeight=0.1601, interpMode=RCIM_Cubic, leaveTangentWeight=0.2668, tangentMode=RCTM_User, time=0.2, value=1}
        {arriveTangentWeight=0.2906, interpMode=RCIM_Cubic, leaveTangentWeight=0.1163, tangentMode=RCTM_User, time=0.7, value=1}
        {arriveTangent=-6.7083, interpMode=RCIM_Cubic, leaveTangent=-6.7083, tangentMode=RCTM_User, time=1, value=0}
       ]}}}
   )
   &SkeletalMeshLocation(
    Bone / Socket Index=$Particles.SpawnGroup
    Bone Sampling Mode=::NewEnumerator4
    MaskSpawnGroup=false
    Position Sampling=::NewEnumerator1
    Sampled Position Offset=$StackContext.ShapeLocation.ShapePosition
    Skeletal Mesh=$System.Skeletal Mesh
    Source Movement Scale:Vector3f=<0 0 0>
   ) {module=SkeletalMeshLocation001}
   &ScaleColor(
    Scale Alpha=&FloatFromCurve(FloatCurve=DataInterface() {propertyValues {Curve {keys [<at 0 1 RCIM_Cubic>,<at 1 0 RCIM_Cubic>]}}}, Scale Curve=2)
    ScaleA=true
    ScaleRGB=true
   )
   &SolveForcesAndVelocity(Clamp Velocity=false, Limit Acceleration=false)
   &DynamicMaterialParameters(
    Index 0 Param 1=2
    Module.Index0Param0Write=true
    Module.Index0Param1Write=true
    Module.Index0Param2Write=true
    Module.Index0Param3Write=true
   )
  ]
  rendererClasses [&NiagaraSpriteRendererProperties]
  renderers [{rendererClass=&NiagaraSpriteRendererProperties}]
 }
 Pixels_Animated_Floating {
  emitterUpdateScript [
   &EmitterState()
   &SpawnRate(SpawnRate=4, Use Spawn Probability=false)
  ]
  particleSpawnScript [
   &PartitionParticles(Number Of Partitions=5)
   &SkeletalMeshLocation(
    Bone / Socket Index=7
    Bone Sampling Mode=::NewEnumerator4
    MaskSpawnGroup=false
    Position Sampling=::NewEnumerator2
    Sampled Position Offset Coordinate Space:ENiagaraSkelMeshTransforms=::NewEnumerator0
    Skeletal Mesh=$System.Skeletal Mesh
   ) {module=SkeletalMeshLocation001}
   &InitializeParticle(
    Lifetime Max=1
    Lifetime Min=0.7
    Lifetime Mode=::NewEnumerator1
    Sprite Size=&Multiply_Vector2D_ByFloat(
     Float=&MakeCustomFloatFromBool(Bool=$User.IsLargeBody, True Float=1.2, False Float=1)
     Vector2D=&Multiply_Vector2D_ByFloat(
      Float=&RandomRangeFloat(Minimum=6, Maximum=10)
      Vector2D=&RandomRangeVector2D(Minimum=<1 1>, Maximum=<1.25 1.5>)
     )
    )
    Sprite Size Mode=::NewEnumerator4
    UsePositionOffset=true
    Write Material Random=true
   )
   /SetVariables() {isSetParameters=true}
   &CameraOffset(Camera Offset Amount=-30)
   &ShapeLocation(
    Apply To Particle Position=::NewEnumerator1
    Box Size=<10 1 2>
    MaskSpawnGroup=false
    Override Seed=false
    Shape Origin=&VectorFromFloat(Value=0)
    Shape Primitive=::NewEnumerator2
    Surface Only=false
    UseOwnerScale=true
   )
   &AddVelocity(
    Rotation Coordinate Space=::Simulation
    Velocity=&RandomRangeVector(Minimum=<0 0 275>, Maximum=<0 0 350>)
    Velocity Speed Scale=&MakeCustomFloatFromBool(Bool=$User.IsLargeBody, True Float=1.5, False Float=1)
   )
   &AddVelocity(
    Module.Use Falloff Distance=false
    Override Seed=false
    Velocity Mode=::NewEnumerator1
    Velocity Origin=$Output.SkeletalMeshLocation001.SampledPositionWithoutOffset
    Velocity Speed=&RandomRangeFloat(Minimum=25, Maximum=50)
   ) {module=AddVelocity001}
  ]
  particleUpdateScript [
   &ParticleState()
   &ScaleSpriteSize(Uniform Curve Sprite Scale=DataInterface() {propertyValues {Curve {keys [{tangentMode=RCTM_User, time=0, value=0}, <at 0.2 0.8>, <at 0.8 0.9>, {tangentMode=RCTM_User, time=1, value=1.3}]}}})
   &Drag(Drag=2.5, Use Linear Drag=true, Use Rotational Drag=true)
   &SkeletalMeshLocation(
    Bone / Socket Index=&SelectIntFromArray(Direct Array Index=$Particles.PartitionParticles.Partition, Int Selection Array=DataInterface() {propertyValues {IntData [0,1,2,3,4]}})
    Bone Sampling Mode=::NewEnumerator4
    MaskSpawnGroup=false
    Orientation Sampling=::NewEnumerator1
    Position Sampling=::NewEnumerator1
    Sampled Position Offset=&AddVectorToPosition(
     Position=$StackContext.ShapeLocation.ShapePosition
     Vector=&Multiply_VectorByFloat(
      Vector=&RandomRangeVector(Minimum=<0 0 10>, Maximum=<0 -15 15>)
      Float=&MakeCustomFloatFromBool(Bool=$User.IsLargeBody, True Float=1.1, False Float=1)
     )
    )
    Sampled Position Offset Coordinate Space:ENiagaraSkelMeshTransforms=::NewEnumerator2
    Skeletal Mesh=$System.Skeletal Mesh
   ) {module=SkeletalMeshLocation003}
   &SolveForcesAndVelocity(Clamp Velocity=false, Limit Acceleration=false)
   &InterpolateOverTime(Rate Of Change=7, Target Position Value:NiagaraPosition=$Output.SolveForcesAndVelocity.Position, Variable Type=::NewEnumerator6)
   &ScaleColor(
    Linear Color Curve=DataInterface() {propertyValues {
      AlphaCurve {keys [<at 0 0>,<at 0.0324 1>,<at 0.9279 1>,<at 1 0>]}
      BlueCurve {keys [<at 0.0382 1>,<at 0.1842 0>,<at 0.5818 0.0173>,<at 0.8528 0.2174>]}
      GreenCurve {keys [<at 0.0382 0.0007>,<at 0.1842 0.0969>,<at 0.5818 0>,<at 0.8528 0>]}
      RedCurve {keys [<at 0.0382 0.5893>,<at 0.1842 1.5>,<at 0.5818 1.1>,<at 0.8528 0.5885>]}
     }}
    Scale Mode=::NewEnumerator2
   )
   /SetVariables(Particles.color2:LinearColor=&MultiplyLinearColor(
     Linear Color=&ColorFromCurve(ColorCurve=DataInterface() {propertyValues {
        AlphaCurve {keys [<at 0 0 RCIM_Cubic>,<at 0.0386 0.9939>,<at 0.885 1>,<at 1 0 RCIM_Cubic>]}
        BlueCurve {keys [
          <at 0.0588 0.3116>
          <at 0.1258 0.7292>
          <at 0.204 1>
          {arriveTangent=-1.209, interpMode=RCIM_Cubic, leaveTangent=-1.209, time=0.413, value=0.8958}
          <at 0.5419 0>
          <at 0.9905 0.0492>
         ]}
        GreenCurve {keys [
          <at 0.0588 0.6823>
          <at 0.1258 0>
          <at 0.204 0.0972>
          {arriveTangent=0.2803, interpMode=RCIM_Cubic, leaveTangent=0.2803, time=0.413, value=0.6975}
          <at 0.5419 0.5052>
          <at 0.9905 0.3177>
         ]}
        RedCurve {keys [<at 0.0588 0>,<at 0.1258 0.5831>,<at 0.204 0>,<at 0.413 0 RCIM_Cubic>,<at 0.5419 0.0813>,<at 0.9905 0>]}
       }})
     Scale Factor=&MakeCustomVector4FromBool(Bool:NiagaraBool=$User.IsFrontEnd, True Vector4:Vector4f=<0.6 0.6 0.6 1>, False Vector4:Vector4f=<1 1 1 1>)
    )) {isSetParameters=true}
   &DynamicMaterialParameters(
    Index 0 Param 1=&FloatFromCurve(FloatCurve=DataInterface() {propertyValues {Curve {defaultValue=1, keys [{tangentMode=RCTM_User, time=0.8, value=0.1}, {tangentMode=RCTM_User, time=1, value=0.5}]}}})
    Index 0 Param 2=&FloatFromCurve(FloatCurve=DataInterface() {propertyValues {Curve {keys [<at 0.85 3>,<at 1 5>]}}})
    Module.Index0Float0Write=true
    Module.Index0Float1Write=true
    Module.Index0Float2Write=true
    Module.Index0Float3Write=true
   )
   /SetVariables(Particles.PivotOffset2:Vector2f=&Add_Vector2(
     A:Vector2f=$Particles.PivotOffset
     B:Vector2f=&MakeVector2D(
      X=&FloatFromCurve(FloatCurve=DataInterface() {propertyValues {Curve {
          defaultValue=1
          keys [
           {arriveTangent=0.4331, interpMode=RCIM_Cubic, leaveTangent=0.4331, tangentMode=RCTM_User, time=0.7, value=0.1}
           {interpMode=RCIM_Cubic, tangentMode=RCTM_User, time=1, value=0.2}
          ]
         }}})
      Y=&FloatFromCurve(FloatCurve=DataInterface() {propertyValues {Curve {
          defaultValue=1
          keys [
           {arriveTangent=0.4564, interpMode=RCIM_Cubic, leaveTangent=0.4564, tangentMode=RCTM_User, time=0.7, value=0.1}
           {arriveTangentWeight=0.3333, interpMode=RCIM_Cubic, tangentMode=RCTM_Break, time=1, value=0.2}
          ]
         }}})
     )
    )) {isSetParameters=true}
   /SetVariables(Particles.Position=$StackContext.InterpolateOverTime.InterpolateOverTime_Position.Moving Average) {isSetParameters=true}
   /SetVariables(Particles.PositionOffset:NiagaraPosition=&AddVectorToPosition(
     Position=$Particles.Position
     Vector=&MakeVector(
      X=&FloatFromCurve(
       FloatCurve=DataInterface() {propertyValues {Curve {keys [<at 0 0 RCIM_Constant>,<at 0.1 0.5 RCIM_Constant>,<at 0.2 1 RCIM_Constant>,<at 0.5 1 RCIM_Constant>,<at 0.6 0.25 RCIM_Constant>,<at 0.7 0 RCIM_Constant>,<at 0.8 -0.625 RCIM_Constant>,<at 0.9 -1 RCIM_Constant>]}}}
       Scale Curve=&SelectFloatFromArray(
        Array Sampling Mode=::NewEnumerator0
        Direct Array Index=?
        Float Selection Array=DataInterface() {propertyValues {FloatData [0,3,-2,0]}}
        Random Seed=0
       )
      )
      Y=0
      Z=&FloatFromCurve(
       FloatCurve=DataInterface() {propertyValues {Curve {keys [<at 0 0 RCIM_Constant>,<at 0.1 0.25 RCIM_Constant>,<at 0.3 0.575 RCIM_Constant>,<at 0.4 0.8 RCIM_Constant>,<at 0.6 1 RCIM_Constant>,<at 0.7 1.2 RCIM_Constant>,<at 0.8 1.375 RCIM_Constant>]}}}
       Scale Curve=&SelectFloatFromArray(
        Array Sampling Mode=::NewEnumerator0
        Direct Array Index=?
        Float Selection Array=DataInterface() {propertyValues {FloatData [2,4,7]}}
        Random Seed=0
       )
      )
     )
    )) {isSetParameters=true}
  ]
  rendererClasses [&NiagaraSpriteRendererProperties]
  renderers [
   {rendererClass=&NiagaraSpriteRendererProperties}
   {rendererClass=&NiagaraSpriteRendererProperties, rendererIndex=1}
  ]
 }
 SpeedLines {
  emitterUpdateScript [
   &EmitterState()
   &SpawnBurst_Instantaneous(Spawn Count=35, Use Spawn Probability=false, UseLoopCountLimit=true)
   &SpawnRate(SpawnRate=40, Use Spawn Probability=false) {enabled=false}
  ]
  particleSpawnScript [
   &InitializeParticle(
    Color=&SelectLinearColorFromArray(
     Array Sampling Mode=::NewEnumerator0
     Direct Array Index=?
     Color Selection Array=DataInterface() {propertyValues {ColorData [<rgba 1.5 0.6821 0.1172 1>, <rgba 1.5 0.2257 0.8263 1>, <rgba 0 1.5 1.2853 1>]}}
     Random Seed=0
    )
    Lifetime=&FloatFromCurve(
     FloatCurve=DataInterface() {propertyValues {Curve {keys [<at 0.2 0.2>,<at 0.7 0.2>,<at 0.8 0.25>]}}}
     CurveIndex=&RandomRangeFloat(Minimum=0, Maximum=1)
    )
    Sprite Size Max=<1 50>
    Sprite Size Min=<0.5 20>
    Sprite Size Mode=::NewEnumerator2
    UsePositionOffset=false
    Write Material Random=true
   )
   &CameraOffset(Camera Offset Amount=-25)
   &ShapeLocation(MaskSpawnGroup=false, Offset Mode=::NewEnumerator0, Override Seed=false, Sphere Radius=1, UseOwnerScale=true) {module=ShapeLocation001}
   &AddVelocity(
    Module.Use Falloff Distance=false
    Override Seed=false
    Velocity Mode=::NewEnumerator1
    Velocity Origin=$StackContext.Position
    Velocity Speed=&RandomRangeFloat(Minimum=200, Maximum=450)
   ) {module=AddVelocity001}
   &SkeletalMeshLocation(
    Bone / Socket Index=&SelectIntFromArray(Direct Array Index=$Particles.PartitionParticles.Partition, Int Selection Array=DataInterface() {propertyValues {IntData [0,1,2,3,4]}})
    Bone Sampling Mode=::NewEnumerator4
    MaskSpawnGroup=false
    Orientation Sampling=::NewEnumerator1
    Sampled Position Offset=&RandomRangeVector(Minimum=<0 0 0>, Maximum=<0 -15 0>)
    Sampled Position Offset Coordinate Space:ENiagaraSkelMeshTransforms=::NewEnumerator2
    Skeletal Mesh=$System.Skeletal Mesh
   ) {enabled=false, module=SkeletalMeshLocation003}
   &ShapeLocation(Box Size=<25 2 3>, MaskSpawnGroup=false, Override Seed=false, Shape Primitive=::NewEnumerator2, Surface Only=false, UseOwnerScale=true) {enabled=false}
   &AddVelocity(Velocity=&RandomRangeVector(Minimum=<0 0 100>, Maximum=<0 0 200>)) {enabled=false}
  ]
  particleUpdateScript [
   &ParticleState()
   &ScaleSpriteSize(
    Non-Uniform Curve Sprite Scale=DataInterface() {propertyValues {
      XCurve {keys [{tangentMode=RCTM_User, time=0, value=1}, {tangentMode=RCTM_User, time=1, value=1}]}
      YCurve {keys [{tangentMode=RCTM_User, time=0, value=0.3}, {tangentMode=RCTM_User, time=1, value=0.825}]}
     }}
    Scale Sprite Size Mode=::NewEnumerator3
   )
   &SkeletalMeshLocation(
    Bone / Socket Index=$Particles.SpawnGroup
    Bone Sampling Mode=::NewEnumerator4
    MaskSpawnGroup=false
    Position Sampling=::NewEnumerator1
    Sampled Position Offset:NiagaraPosition=$Particles.Initial.ShapeLocation001.ShapePosition
    Skeletal Mesh=$System.Skeletal Mesh
    Source Movement Scale:Vector3f=<0 0 0>
   ) {module=SkeletalMeshLocation001}
   &ScaleColor(
    Linear Color Curve=DataInterface() {propertyValues {
      AlphaCurve {keys [<at 0 0>,<at 0.0728 0.8>,<at 0.1789 0.5>,<at 1 0>]}
      BlueCurve {keys [<at 0 1>,<at 0.8464 0.4485>]}
      GreenCurve {keys [<at 0 1>,<at 0.8464 0.1254>]}
      RedCurve {keys [<at 0 1>,<at 0.8464 1>]}
     }}
    Scale Mode=::NewEnumerator2
   )
   &Drag(Drag=3, Use Linear Drag=true, Use Rotational Drag=true)
   &SolveForcesAndVelocity(Clamp Velocity=false, Limit Acceleration=false)
   &ScaleSpriteSizeBySpeed(Max Scale Factor=<1.2 3>, Min Scale Factor=<1 1.4>)
  ]
  rendererClasses [&NiagaraSpriteRendererProperties]
  renderers [{rendererClass=&NiagaraSpriteRendererProperties}]
 }
}
summary {
 systemName=NS_Aura_Pixel_Start
 emitters [
  {emitterName=Flare, rendererClasses [&NiagaraSpriteRendererProperties]}
  {emitterName=Dark_Glow, rendererClasses [&NiagaraSpriteRendererProperties]}
  {emitterName=Corona, rendererClasses [&NiagaraSpriteRendererProperties]}
  {emitterName=Pixel_small_burst, rendererClasses [&NiagaraSpriteRendererProperties]}
  {emitterName=SpeedLines, rendererClasses [&NiagaraSpriteRendererProperties]}
  {emitterName=Pixel_Burst_Animated, rendererClasses [&NiagaraSpriteRendererProperties]}
  {bEnabled=false, emitterName=Pixel_Burst_Animated001, rendererClasses [&NiagaraSpriteRendererProperties]}
  {emitterName=Pixels_Animated_Floating, rendererClasses [&NiagaraSpriteRendererProperties]}
 ]
 userVariables [
  User.IsFrontEnd {type=NiagaraBool}
  User.IsLargeBody {type=NiagaraBool}
 ]
}
